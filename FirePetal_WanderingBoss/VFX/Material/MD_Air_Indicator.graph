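# Translucent unlit deferred decal material for the FirePetal wandering boss's air indicator VFX, a fiery animated radial shape.
# Radial UVs, distorted by panning noise, scroll soft noise for red-to-orange emissive and plasma for the opacity shape.
# A progress offset shifts that shape through a curve-atlas gradient, with GoldSand sparkle and a circular mask.
# Decal color channels drive it per instance: red progress, green brightness, blue and alpha opacity.
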Material /FirePetal_WanderingBoss/VFX/Material/MD_Air_Indicator
graph {
 properties {
  bUsedWithNiagaraSprites=true
  blendMode=BLEND_Translucent
  materialDomain=MD_DeferredDecal
  pixelDepthOffsetMode=PDOM_Legacy
  shadingModel=MSM_Unlit
  usedShadingModels=Unlit
 }
 nodes [
  LinearInterpolate(
   A=Constant() {r=-0.5}
   B=Constant() {r=-0.45}
   Alpha=MaterialFunctionCall(Speed=Constant() {r=1.2}) {materialFunction=&DebugTimeSine}.Result
  ) {constA=-0.5, constAlpha=0.3, constB=0.5}
  n1 = AppendVector(A=Constant() {r=2}, B=Constant() {r=0.5})
  n2 = MaterialFunctionCall() {materialFunction=&VectorToRadialValue}
  n3 = AppendVector(A=Constant() {r=2}, B=Constant() {r=0.4})
  n4 = TextureCoordinate()
  n5 = TextureSample!(
   UVs=Panner(Coordinate=Multiply(A=@n2.Radial Coordinates, B=@n3)) {speedY=-0.3}
   Tex
   "DDX(UVs)"=DDX(Value=@n4)
   "DDY(UVs)"=DDY(Value=@n4)
   Apply View MipBias
  ) {mipValueMode=TMVM_Derivative, samplerType=SAMPLERTYPE_LinearColor, texture=&T_NoisePack_LinearColor}
  n6 = Divide(A=ComponentMask(None=@n5.RGB) {g=true, r=true}, B=Constant() {r=6})
  n7 = Panner(Coordinate=Add(A=Multiply(A=@n2.Radial Coordinates, B=@n1), B=@n6)) {speedY=0.5}
  n8 = TextureCoordinate()
  n9 = DDX(Value=@n8)
  n10 = DDY(Value=@n8)
  n11 = MaterialFunctionCall(Float2=@n2.Radial Coordinates) {materialFunction=&BreakOutFloat2Components}
  n12 = Multiply(B=@n11.G) {constA=1}
  n13 = Multiply(
   A=LinearInterpolate(Alpha=TextureSample(UVs=@n7, "DDX(UVs)"=@n9, "DDY(UVs)"=@n10) {mipValueMode=TMVM_Derivative, texture=/Game/Effects/Fort_Effects/Emotes/UltraLight/Texture/T_SoftNoise_BC7}.R) {constA=0.2}
   B=Saturate(None=Multiply(A=@n12) {constB=4})
  )
  n14 = Multiply(
   A=LinearInterpolate(
    A=VectorParameter() {parameterName=Param_1, defaultValue=<rgba 1 0.07 0 1>}.RGB
    B=VectorParameter() {defaultValue=<rgba 5 1.9478 0 1>}.RGB
    Alpha=@n13
   )
   B=@n13
  ) {constA=1}
  n15 = ScalarParameter() {defaultValue=1, parameterName=ColorScale}
  n16 = DecalColor()
  MaterialFunctionCall(Emissive=Multiply(A=Multiply(A=@n14, B=@n15), B=@n16.G)) {materialFunction=&MF_FX_EmissiveAdjustments}
  n17 = MaterialFunctionCall!(
   GoldSand_U_Tile=ScalarParameter() {defaultValue=10, parameterName=GoldSand_U_Tile}
   GoldSand_V_Tile=ScalarParameter() {defaultValue=4, parameterName=GoldSand_V_Tile}
   GoldSandIntensity=ScalarParameter() {defaultValue=10, parameterName=GoldSandIntensity}
   PannerSpeedU=Constant()
   PannerSpeedV=Constant() {r=-1}
   PannerTimeInput=Time()
  ) {materialFunction=/Game/Effects/Niagara/MasterMaterials/Functions/MF_GoldSand}
  n18 = TextureSample(UVs=@n7, "DDX(UVs)"=@n9, "DDY(UVs)"=@n10) {mipValueMode=TMVM_Derivative, samplerType=SAMPLERTYPE_LinearColor, texture=/Game/Effects/Fort_Effects/Emotes/UltraLight/Texture/T_Plasma_A_C}
  n19 = Add(A=Divide(A=@n18.R, B=Constant() {r=5}), B=Multiply(A=ComponentMask(None=@n6) {r=true}) {constB=2})
  n20 = Constant() {r=2500}
  n21 = DecalColor()
  n22 = ScalarParameter() {parameterName=ConstantProgress}
  n23 = LinearInterpolate(Alpha=LinearInterpolate(
    A=MaterialFunctionCall(Distance=@n20) {materialFunction=&Near_Camera_Fade}.Result
    B=Add(A=@n21.R, B=@n22)
    Alpha=ScalarParameter() {defaultValue=1, parameterName=UseConstantProgress}
   )) {constA=-1}
  n24 = Subtract(A=Add(A=Add(A=@n19, B=@n11.G), B=Multiply(B=@n23) {constA=-0.5}) {constB=0}) {constB=0.5}
  n25 = ScalarParameter() {defaultValue=1, parameterName=f132354}
  n26 = CurveAtlasRowParameter(CurveTime=Add(A=Multiply(A=@n24, B=@n25)) {constB=0.5}) {
   atlas=/Game/Effects/Fort_Effects/Materials/Energy/CA_FX_Energy
   curve=/Game/Effects/Fort_Effects/Materials/Energy/Curve_FX_Energy_01
   parameterName=GradientCurve
  }
  n27 = ChannelMaskParameter(None=MaterialFunctionCall(R=@n26.R, G=@n26.R, B=@n26.R, A=@n26.R) {materialFunction=&AppendMany}.RGBA) {maskChannel=Alpha, parameterName=GradientChannelMask, defaultValue=<rgba 0 0 0 1>}
  n28 = Multiply(
   A=SphereMask(A=TextureCoordinate(), B=Constant() {r=0.5}) {attenuationRadius=0.5, hardnessPercent=50}
   B=Add(A=@n27.R, B=Multiply(A=@n27.R, B=@n17.Result))
  ) {constA=1}
  n29 = ScalarParameter() {defaultValue=1, parameterName=OpacityScale}
  n30 = DecalColor()
  Multiply(A=@n16.A, B=Multiply(A=Multiply(A=@n28) {constB=0.4}, B=Multiply(A=@n29, B=@n30.B)))
 ]
}
