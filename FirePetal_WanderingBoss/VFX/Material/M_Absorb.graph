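# Translucent, unlit, two-sided Niagara sprite and mesh material for the FirePetal wandering boss's absorb VFX.
# A panning plasma texture, distorted by scrolling noise, uses UVs with a power curve that warps U.
# Its normalized color is tinted by particle color, with hue shifted along U, for emissive.
# An opacity mask multiplies particle alpha, axis-aligned Fresnel, a U edge mask, refined texture luminance and depth fade.
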Material /FirePetal_WanderingBoss/VFX/Material/M_Absorb
graph {
 properties {
  bAutomaticallySetUsageInEditor=false
  bUsedWithNiagaraMeshParticles=true
  bUsedWithNiagaraSprites=true
  blendMode=BLEND_Translucent
  shadingModel=MSM_Unlit
  twoSided=true
  usedShadingModels=Unlit
 }
 nodes [
  n1 = ScalarParameter() {defaultValue=-0.2, parameterName=dw}
  n2 = MaterialFunctionCall(Float2=TextureCoordinate()) {materialFunction=&BreakOutFloat2Components}
  n3 = ParticleColor()
  n4 = TextureObjectParameter() {
   group=MainTexture
   parameterName=MainTexture_Texture
   samplerType=SAMPLERTYPE_LinearColor
   sortPriority=0
   texture=/Game/Effects/Fort_Effects/Emotes/UltraLight/Texture/T_Plasma_A_C
  }
  n5 = TextureObjectParameter() {
   group=Noise
   parameterName=Noise_Texture
   samplerType=SAMPLERTYPE_LinearGrayscale
   sortPriority=0
   texture=/Game/Effects/Fort_Effects/Emotes/UltraLight/Texture/T_Tiling_Noise_01_GreyScale
  }
  n6 = VectorParameter() {group=Noise, parameterName=Noise_UVControl, channelNames {a=VPan, b=UPan, g=VScale, r=UScale}, defaultValue=<rgba 0.3 1 -0.5 0>}
  n7 = MaterialFunctionCall(Float2=TextureCoordinate()) {materialFunction=&BreakOutFloat2Components}
  n8 = OneMinus(None=@n7.R)
  n9 = ScalarParameter() {defaultValue=1.5, parameterName=UVPower}
  n10 = MaterialFunctionCall(X=OneMinus(None=Power(Base=@n8, Exp=@n9)), Y=@n7.G) {materialFunction=&MakeFloat2}
  n11 = VectorParameter() {
   group=Noise
   parameterName=Noise_NoiseControl
   sortPriority=60
   channelNames {a=VStrength, b=UStrength, g=VOffset, r=UOffset}
   defaultValue=<rgba -0.5 -0.5 0 -0.3>
  }
  n12 = VectorParameter() {
   group=MainTexture
   parameterName=MainTexture_UVControl
   channelNames {a=VPan, b=UPan, g=VScale, r=UScale}
   defaultValue=<rgba 0.3 1 -0.125 0>
  }
  n13 = MaterialFunctionCall(
   Texture=@n4
   AdditionalUVOffset=MaterialFunctionCall!(
    NoiseValue=MaterialFunctionCall(Texture=@n5, UVControl_RGB=@n6, UVControl_A=@n6, "[Override]UVInput"=@n10.Result) {materialFunction=&FXCommon_TextureSample}.RG
    Control_RGB=@n11
    Control_A=@n11
   ) {materialFunction=/Game/Effects/Niagara/MasterMaterials/Functions/Common/FXCommon_NoiseControl}.NoiseValue
   UVControl_RGB=@n12
   UVControl_A=@n12
   "[Override]UVInput"=@n10.Result
  ) {materialFunction=&FXCommon_TextureSample}
  n14 = Normalize(VectorInput=@n13.RGB)
  MaterialFunctionCall(Emissive=MaterialFunctionCall(Hue Shift Percentage=Multiply(A=@n1, B=@n2.R), Texture=Multiply(A=@n3.RGB, B=@n14)) {materialFunction=&HueShift}.Result) {materialFunction=&MF_FX_EmissiveAdjustments}
  n15 = MaterialFunctionCall(Axis=Constant3Vector() {constant=<rgba 0 0 1 0>}, Exponent=ScalarParameter() {defaultValue=5, parameterName=EX}) {materialFunction=&AxisAlignedFresnel}
  n16 = VectorParameter() {group=Fade, parameterName=Fade_EdgeMask_InvertControl, channelNames {a="", b=Power, g=Multiply, r=Offset}, defaultValue=<rgba 0 15 2 1>}
  n17 = VectorParameter() {group=Fade, parameterName=Fade_EdgeMask_Control, channelNames {a="", b=Power, g=Multiply, r=Offset}, defaultValue=<rgba 0 1.2 2.5 1>}
  n18 = Constant() {r=1}
  n19 = VectorParameter() {parameterName=MainTexture_Refine, defaultValue=<rgba 5 10 1 0.5>}
  n20 = Multiply(
   A=MaterialFunctionCall!(UorV=@n7.R, Invert_EdgeControl=@n16, NoInvert_EdgeControl=@n17) {materialFunction=/Game/Effects/Niagara/MasterMaterials/Functions/Common/FXCommon_EdgeMask_Ex}.Mask
   B=MaterialFunctionCall(Texture=Desaturation(None=@n13.RGB, Fraction=@n18), RGB=@n19.RGB, A=@n19.RGB) {materialFunction=&MF_FX_Refine}.Result
  )
  n21 = ScalarParameter() {defaultValue=300, parameterName=DF}
  Multiply(A=Multiply(A=@n3.A, B=Multiply(A=@n15.Fresnel, B=@n20)), B=Power(Base=DepthFade(FadeDistance=@n21)))
 ]
}
