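# Unlit translucent Niagara sprite material for the Wandering Boss's DiamondSkin weakpoint flare.
# Panning cloud noise distorts scaled and offset polar UVs, producing an animated radial gradient.
# That gradient, refined and tinted by particle color, feeds the emissive adjustment.
# An opacity-style chain multiplies the powered, tiling-clamped gradient by particle alpha, depth fade, panning Perlin noise and a polar edge mask.
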Material /FirePetal_WanderingBoss/VFX/Material/M_DiamondSkin_Weakpoint_Flare
graph {
 properties {
  bUseTranslucencyVertexFog=false
  bUsedWithNiagaraSprites=true
  blendMode=BLEND_Translucent
  pixelDepthOffsetMode=PDOM_Legacy
  shadingModel=MSM_Unlit
  usedShadingModels=Unlit
 }
 nodes [
  n1 = TextureObjectParameter() {
   group=FlowTexture
   parameterName=FlowTexture_Texture
   samplerType=SAMPLERTYPE_LinearGrayscale
   sortPriority=0
   texture=/Game/Effects/Fort_Effects/Emotes/UltraLight/Texture/T_Cloud_Greyscale
  }
  n2 = VectorParameter() {
   group=FlowTexture
   parameterName=FlowTexture_UVControl
   channelNames {a=VPan, b=UPan, g=VScale, r=UScale}
   defaultValue=<rgba 0.2 0.2 -0.5 0>
  }
  n3 = MaterialFunctionCall!(UV Coord=TextureCoordinate(), LenghtScale, RadialScale) {materialFunction=/Game/Global/MaterialFunctions/TechArt/MF_PolarCoordinates}
  n4 = VectorParameter() {
   group=FlowNoise
   parameterName=FlowNoise_NoiseControl
   sortPriority=60
   channelNames {a=VStrength, b=UStrength, g=VOffset, r=UOffset}
   defaultValue=<rgba -0.466 -0.5 0.35 0>
  }
  n5 = VectorParameter() {parameterName=AA, defaultValue=<rgba 1 0 -0.2368 1>}
  n6 = AppendVector(A=@n5.R, B=@n5.R)
  n7 = Add(
   A=MaterialFunctionCall!(
    NoiseValue=MaterialFunctionCall(Texture=@n1, UVControl_RGB=@n2, UVControl_A=@n2, "[Override]UVInput"=@n3.Result) {materialFunction=&FXCommon_TextureSample}.RG
    Control_RGB=@n4
    Control_A=@n4
   ) {materialFunction=/Game/Effects/Niagara/MasterMaterials/Functions/Common/FXCommon_NoiseControl}.NoiseValue
   B=Add(A=Multiply(A=@n3.Result, B=@n6), B=AppendVector(A=@n5.B, B=@n5.B))
  )
  n8 = OneMinus(None=@n7)
  LinearInterpolate(
   A=VectorParameter() {parameterName=ccc, defaultValue=<rgba 3 0.15 0.15 1>}.RGB
   Alpha=Power(Base=@n8, Exp=ScalarParameter() {defaultValue=35, parameterName=ssE2})
  )
  n9 = VectorParameter() {parameterName=Refine, defaultValue=<rgba 45 2 0.25 0.5>}
  n10 = MaterialFunctionCall(Texture=@n8, RGB=@n9.RGB, A=@n9.RGB) {materialFunction=&MF_FX_Refine}
  n11 = ParticleColor()
  MaterialFunctionCall(Emissive=Multiply(A=@n10.Result, B=@n11.RGB)) {materialFunction=&MF_FX_EmissiveAdjustments}
  n12 = Multiply(
   A=Power(Base=@n8, Exp=ScalarParameter() {defaultValue=5, parameterName=ssE})
   B=MaterialFunctionCall!(TexCoord=@n7, ClampChannel=Constant2Vector() {r=1}) {materialFunction=/Game/Effects/Fort_Effects/MaterialFunctions/MF_FX_TilingMethod}.Result
  )
  n13 = ScalarParameter() {parameterName=DepthFade}
  n14 = Multiply(B=Multiply(A=Multiply(A=@n11.A), B=DepthFade(FadeDistance=@n13))) {constA=1}
  n15 = TextureObjectParameter() {
   group=Mask
   parameterName=Mask_Texture
   samplerType=SAMPLERTYPE_LinearGrayscale
   sortPriority=0
   texture=&T_SimplePerlin_Greyscale
  }
  n16 = VectorParameter() {group=Mask, parameterName=Mask_UVControl, channelNames {a=VPan, b=UPan, g=VScale, r=UScale}, defaultValue=<rgba 0.5 1 0.2 -0.5>}
  n17 = MaterialFunctionCall(Texture=@n15, UVControl_RGB=@n16, UVControl_A=@n16, "[Override]UVInput"=@n3.Result) {materialFunction=&FXCommon_TextureSample}
  n18 = ScalarParameter() {defaultValue=2, parameterName=EXX}
  n19 = VectorParameter() {
   group=MaskShape
   parameterName=MaskShape_EdgeMask_InvertControl
   channelNames {a="", b=Power, g=Multiply, r=Offset}
   defaultValue=<rgba -0.5 3.8859 2 1>
  }
  n20 = VectorParameter() {group=MaskShape, parameterName=MaskShape_EdgeMask_Control, channelNames {a="", b=Power, g=Multiply, r=Offset}, defaultValue=<rgba 15 5 2 1>}
  Multiply(
   A=MaterialFunctionCall!(
    UorV=MaterialFunctionCall(Float2=@n3.Result) {materialFunction=&BreakOutFloat2Components}.R
    Invert_EdgeControl=@n19
    NoInvert_EdgeControl=@n20
   ) {materialFunction=/Game/Effects/Niagara/MasterMaterials/Functions/Common/FXCommon_EdgeMask_Ex}.Mask
   B=Multiply(A=Multiply(A=@n12, B=@n14), B=Power(Base=@n17.R, Exp=@n18))
  )
  Normalize(VectorInput=@n10.Result)
  TextureCoordinate()
 ]
}
