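# Additive, unlit, two-sided Niagara sprite and mesh particle material, named as an explosion wave effect for the Wandering Boss.
# Particle color feeds a shared emissive adjustment function.
# A radial-coordinate component drives an edge-mask function with parameterized normal and inverted controls, shaping the wave procedurally without textures in this graph.
# That mask is multiplied by particle alpha and a powered depth fade to soften intersections with scene geometry.
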Material /FirePetal_WanderingBoss/VFX/Material/M_ExplosionWave
graph {
 properties {
  bAutomaticallySetUsageInEditor=false
  bUsedWithNiagaraMeshParticles=true
  bUsedWithNiagaraSprites=true
  blendMode=BLEND_Additive
  shadingModel=MSM_Unlit
  twoSided=true
  usedShadingModels=Unlit
 }
 nodes [
  n1 = ParticleColor()
  MaterialFunctionCall(Emissive=@n1.RGB) {materialFunction=&MF_FX_EmissiveAdjustments}
  n2 = ScalarParameter() {defaultValue=300, parameterName=DF}
  n3 = MaterialFunctionCall(Float2=MaterialFunctionCall() {materialFunction=&VectorToRadialValue}.Radial Coordinates) {materialFunction=&BreakOutFloat2Components}
  n4 = VectorParameter() {group=E, parameterName=E_EdgeMask_InvertControl, channelNames {a="", b=Power, g=Multiply, r=Offset}, defaultValue=<rgba 0 3.5098 2 1>}
  n5 = VectorParameter() {group=E, parameterName=E_EdgeMask_Control, channelNames {a="", b=Power, g=Multiply, r=Offset}, defaultValue=<rgba -0.1536 1 1.8321 1>}
  Multiply(
   A=Power(Base=DepthFade(FadeDistance=@n2))
   B=Multiply(
    A=@n1.A
    B=MaterialFunctionCall!(UorV=@n3.G, Invert_EdgeControl=@n4, NoInvert_EdgeControl=@n5) {materialFunction=/Game/Effects/Niagara/MasterMaterials/Functions/Common/FXCommon_EdgeMask_Ex}.Mask
   )
  )
 ]
}
