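# Translucent, unlit, two-sided Niagara material for a ground VFX in the FirePetal wandering boss set, usable on sprites, ribbons and mesh particles.
# A panning noise texture distorts the UVs, and the distorted U coordinate is contrast-boosted into a gradient.
# That gradient is multiplied by powered depth fade, inverted Fresnel, a U-axis edge mask and particle alpha to form an alpha mask.
# Particle color feeds a shared emissive adjustment function.
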Material /FirePetal_WanderingBoss/VFX/Material/M_GroundOut
graph {
 properties {
  bAutomaticallySetUsageInEditor=false
  bUsedWithNiagaraMeshParticles=true
  bUsedWithNiagaraRibbons=true
  bUsedWithNiagaraSprites=true
  blendMode=BLEND_Translucent
  shadingModel=MSM_Unlit
  twoSided=true
  usedShadingModels=Unlit
 }
 nodes [
  Constant()
  n1 = ParticleColor()
  MaterialFunctionCall(Emissive=@n1.RGB) {materialFunction=&MF_FX_EmissiveAdjustments}
  n2 = TextureObjectParameter() {
   group=Noise
   parameterName=Noise_Texture
   samplerType=SAMPLERTYPE_LinearGrayscale
   sortPriority=0
   texture=/Game/Effects/Fort_Effects/Emotes/UltraLight/Texture/T_DirectionalNoise_Greyscale
  }
  n3 = VectorParameter() {group=Noise, parameterName=Noise_UVControl, channelNames {a=VPan, b=UPan, g=VScale, r=UScale}, defaultValue=<rgba 0.2 3 2 0>}
  n4 = VectorParameter() {
   group=Noise
   parameterName=Noise_NoiseControl
   sortPriority=60
   channelNames {a=VStrength, b=UStrength, g=VOffset, r=UOffset}
   defaultValue=<rgba 0 -0.5 0.5 0>
  }
  n5 = MaterialFunctionCall!(
   NoiseValue=MaterialFunctionCall(Texture=@n2, UVControl_RGB=@n3, UVControl_A=@n3) {materialFunction=&FXCommon_TextureSample}.RG
   Control_RGB=@n4
   Control_A=@n4
  ) {materialFunction=/Game/Effects/Niagara/MasterMaterials/Functions/Common/FXCommon_NoiseControl}
  n6 = TextureCoordinate()
  n7 = MaterialFunctionCall(Float2=MaterialFunctionCall(XY=Add(A=@n5.NoiseValue, B=@n6)) {materialFunction=&Swizzle}.YX) {materialFunction=&BreakOutFloat2Components}
  n8 = ScalarParameter() {defaultValue=5, parameterName=dqw}
  n9 = ScalarParameter() {defaultValue=100, parameterName=DF}
  n10 = Multiply(
   A=Multiply(
    A=MaterialFunctionCall(In=@n7.G, Contrast=@n8) {materialFunction=&CheapContrast}.Result
    B=ScalarParameter() {defaultValue=1, parameterName=SO}
   ) {constB=1.5}
   B=Power(Base=DepthFade(FadeDistance=@n9), Exp=ScalarParameter() {defaultValue=1.2, parameterName=DFP}) {constExponent=1.5}
  )
  n11 = Power(Base=OneMinus(None=Fresnel() {baseReflectFraction=0}))
  n12 = MaterialFunctionCall(Float2=TextureCoordinate()) {materialFunction=&BreakOutFloat2Components}
  n13 = VectorParameter() {group=Mask, parameterName=Mask_EdgeMask_InvertControl, channelNames {a="", b=Power, g=Multiply, r=Offset}, defaultValue=<rgba 0 5 1 1>}
  n14 = VectorParameter() {group=Mask, parameterName=Mask_EdgeMask_Control, channelNames {a="", b=Power, g=Multiply, r=Offset}, defaultValue=<rgba 1 1 1 1>}
  Multiply(
   A=Multiply(
    A=Multiply(A=@n10, B=@n11)
    B=MaterialFunctionCall!(UorV=@n12.R, Invert_EdgeControl=@n13, NoInvert_EdgeControl=@n14) {materialFunction=/Game/Effects/Niagara/MasterMaterials/Functions/Common/FXCommon_EdgeMask_Ex}.Mask
   )
   B=@n1.A
  )
 ]
}
