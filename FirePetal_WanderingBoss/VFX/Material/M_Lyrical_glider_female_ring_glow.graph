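# Unlit translucent Niagara material for the Wandering Boss's female Lyrical glider ring glow, projected decal-style onto geometry behind it.
# Scene position behind translucency, relative to the particle, yields planar UVs sized by a dynamic parameter and a height fade.
# Emissive combines particle color, time-of-day tint, optional noise and desaturation with sphere-masked, bump-offset crack texture channels in separate colors.
# Opacity combines radial mask, particle alpha and height fade; a camera-direction term computes a depth-based offset.
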Material /FirePetal_WanderingBoss/VFX/Material/M_Lyrical_glider_female_ring_glow
graph {
 properties {
  bUsedWithNiagaraMeshParticles=true
  bUsedWithNiagaraSprites=true
  blendMode=BLEND_Translucent
  pixelDepthOffsetMode=PDOM_Legacy
  shadingModel=MSM_Unlit
  twoSided=true
  usedShadingModels=Unlit
 }
 nodes [
  n1 = Constant3Vector() {constant=<rgba 1 1 1 0>}
  n2 = NamedRerouteUsage()
  n3 = ParticleRandom()
  n4 = LinearInterpolate(
   A=Constant() {r=1}
   B=TextureSample(UVs=Add(A=@n2, B=@n3)) {samplerType=SAMPLERTYPE_Grayscale, texture=&LowResBlurredNoise}.R
   Alpha=ScalarParameter() {parameterName=Use_gradients_Texture}
  )
  n5 = ParticleColor()
  n6 = Multiply(
   A=MaterialFunctionCall(
    Color_Morning=@n1
    Color_Day=@n1
    Color_Evening=@n1
    Color_Night=VectorParameter() {parameterName=Color_Night, defaultValue=<rgba 0.7955 1 0.2496 0>}.RGB
   ) {materialFunction=&Color_Time_of_Day}.Result
   B=LinearInterpolate(A=Multiply(A=@n4, B=@n5.RGB), B=ParticleColor().RGB, Alpha=ScalarParameter() {parameterName=Use simple particlecolor})
  )
  n7 = DynamicParameter() {defaultValue=<rgba 1 1 1 0>, paramNames ["", "", Param3, Des]}
  n8 = Desaturation(None=@n6, Fraction=@n7.Des)
  n9 = ScalarParameter() {parameterName=Use Desaturation}
  n10 = NamedRerouteUsage()
  n11 = Power(Base=SphereMask(A=@n10, B=Constant() {r=0.5}) {attenuationRadius=0.3, hardnessPercent=0})
  n12 = TextureObjectParameter() {
   group=Crack
   parameterName=Crack_Texture
   samplerType=SAMPLERTYPE_LinearColor
   sortPriority=0
   texture=&/T_GroundCrack
  }
  n13 = ScalarParameter() {defaultValue=-10, parameterName=testa}
  n14 = VectorParameter() {group=Crack, parameterName=Crack_UVControl, channelNames {a=VPan, b=UPan, g=VScale, r=UScale}, defaultValue=<rgba 1 1 0 0>}
  n15 = MaterialFunctionCall(
   Texture=@n12
   AdditionalUVOffset=Add(A=@n7.G, B=BumpOffset(Coordinate=@n10, Height=@n13))
   UVControl_RGB=@n14
   UVControl_A=@n14
   "[Override]UVInput"=NamedRerouteUsage()
  ) {materialFunction=&FXCommon_TextureSample}
  n16 = VectorParameter() {parameterName=TestC2, defaultValue=<rgba 20 1.6667 0 1>}
  n17 = VectorParameter() {parameterName=TestC, defaultValue=<rgba 50 4.1667 0 1>}
  n18 = Add(
   A=Multiply(A=@n15.B, B=VectorParameter() {parameterName=TestC4, defaultValue=<rgba 10 0.1667 0 1>}.RGB)
   B=Add(A=Multiply(A=@n15.G, B=@n16.RGB), B=Multiply(A=@n17.RGB, B=@n15.R))
  )
  Add(
   A=MaterialFunctionCall(Emissive=LinearInterpolate(A=@n6, B=@n8, Alpha=@n9)) {materialFunction=&MF_FX_EmissiveAdjustments}.Result
   B=Multiply(A=Multiply(A=@n11, B=@n18), B=DynamicParameter() {defaultValue=<rgba 1 1 1 0>, paramNames ["", "", Param3, Des]}.R)
  )
  n19 = Normalize(VectorInput=MaterialFunctionCall() {materialFunction=&CameraDirectionVector}.Result)
  n20 = DotProduct(A=Subtract(A=CameraPositionWS(), B=WorldPosition() {worldPositionShaderOffset=WPT_ExcludeAllShaderOffsets}.XYZ), B=@n19)
  n21 = MaterialFunctionCall(Float2=ViewProperty().Property) {materialFunction=&BreakOutFloat2Components}
  n22 = Multiply(B=Divide(A=@n21.R, B=@n21.R)) {constA=10}
  Multiply(A=Max(A=Add(A=@n20, B=@n22) {constB=11}) {constB=0}, B=@n19)
  n23 = ParticlePositionWS()
  n24 = TransformPosition(None=Constant3Vector()) {transformType=TRANSFORMPOSSOURCE_World}
  n25 = TransformPosition(None=Subtract(
    A=MaterialFunctionCall() {materialFunction=&WorldPositionBehindTranslucency}.Result
    B=Subtract(A=@n23, B=@n24)
   )) {transformSourceType=TRANSFORMPOSSOURCE_World}
  n26 = ComponentMask(None=@n25) {b=true}
  n27 = ScalarParameter() {defaultValue=-21.7433, parameterName=ZFadeOffset}
  n28 = Saturate(None=Divide(A=Add(A=@n26, B=@n27), B=ScalarParameter() {defaultValue=50, parameterName=ZFade}) {constB=150})
  n29 = NamedRerouteUsage()
  n30 = Constant() {r=0.5}
  n31 = LinearInterpolate(
   A=Power(Base=SphereMask(A=@n29, B=@n30) {attenuationRadius=0.5, hardnessPercent=0}) {constExponent=0.8}
   B=Constant() {r=1}
   Alpha=ScalarParameter() {parameterName=Use_gradients_Texture}
  )
  Multiply(
   A=Power(Base=OneMinus(None=@n28), Exp=ScalarParameter() {defaultValue=5, parameterName=ZfadePow})
   B=LinearInterpolate(A=Multiply(A=@n5.A, B=@n31), B=ParticleColor().A, Alpha=ScalarParameter() {parameterName=Use simple particlecolor})
  )
  n32 = ComponentMask(None=@n25) {g=true, r=true}
  n33 = DynamicParameter() {parameterIndex=1, defaultValue=<rgba 1 1 1 0>, paramNames [Sx,Sy,Param3,Des]}
  n34 = AppendVector(A=@n33.Sx, B=@n33.Sx)
  NamedRerouteDeclaration(None=Add(A=Divide(A=@n32, B=@n34)) {constB=0.5}) {name=WUV, nodeColor=<rgba 0.4531 0 1 1>}
  TextureCoordinate()
  TextureCoordinate() {uTiling=0.5, vTiling=0.5}
 ]
}
