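# Translucent unlit Niagara sprite material for the Fire Petal wandering boss's weakpoint line: a bright tapered streak with a soft glow.
# A V-based taper, blendable from single-tip to double-tip sine, scales U around the center, and mirrored gradients shape the hot core and wider glow.
# A panning plasma texture, raised to separate powers, breaks up both layers.
# Particle color tints the boosted core for emissive, while particle alpha scales the combined core and glow for opacity.
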Material /FirePetal_WanderingBoss/VFX/Material/M_WeakpointLine
graph {
 properties {
  bAutomaticallySetUsageInEditor=false
  bEnableResponsiveAA=true
  bUsedWithNiagaraSprites=true
  blendMode=BLEND_Translucent
  shadingModel=MSM_Unlit
  twoSided=true
  usedShadingModels=Unlit
 }
 nodes [
  n1 = TextureCoordinate()
  n2 = MaterialFunctionCall(Float2=@n1) {materialFunction=&BreakOutFloat2Components}
  n3 = ScalarParameter() {defaultValue=1, parameterName=ThickGradMul}
  n4 = Multiply(A=@n2.G) {constB=0.5}
  n5 = Saturate(None=LinearInterpolate(A=Multiply(A=@n2.G, B=@n3), B=Sine(None=@n4), Alpha=ScalarParameter() {parameterName=UseDubleTip}))
  n6 = Constant() {r=1}
  n7 = MaterialFunctionCall(Float2=MaterialFunctionCall(
    UVs=@n1
    Texture Scale=MaterialFunctionCall(X=@n5, Y=@n6) {materialFunction=&MakeFloat2}.Result
   ) {materialFunction=&ScaleUVsByCenter}.UVs) {materialFunction=&BreakOutFloat2Components}
  n8 = MaterialFunctionCall!(TexCoord=@n7.R, ClampChannel=Constant2Vector() {g=1, r=1}) {materialFunction=/Game/Effects/Fort_Effects/MaterialFunctions/MF_FX_TilingMethod}
  n9 = Multiply(A=@n7.R) {constB=2}
  n10 = Frac(None=@n9)
  n11 = LinearInterpolate(A=@n10, B=OneMinus(None=@n10), Alpha=Floor(None=@n9))
  n12 = MaterialFunctionCall(
   In=Power(Base=Multiply(A=@n8.Result, B=@n11), Exp=ScalarParameter() {defaultValue=25, parameterName=CenterThickPower})
   Contrast=ScalarParameter() {defaultValue=0.5, parameterName=CenterContrast}
  ) {materialFunction=&CheapContrast}
  n13 = TextureObjectParameter() {
   group=TextureA
   parameterName=TextureA_Texture
   samplerType=SAMPLERTYPE_LinearColor
   sortPriority=0
   texture=/Game/Items/Traps/Textures/T_Plasma_Flames_01_M
  }
  n14 = VectorParameter() {group=TextureA, parameterName=TextureA_UVControl, channelNames {a=VPan, b=UPan, g=VScale, r=UScale}, defaultValue=<rgba 0 0.7 0 -0.3>}
  n15 = MaterialFunctionCall(Texture=@n13, UVControl_RGB=@n14, UVControl_A=@n14) {materialFunction=&FXCommon_TextureSample}
  n16 = Power(Base=@n15.R, Exp=ScalarParameter() {defaultValue=2, parameterName=TexAToCenterPower})
  n17 = Multiply(A=@n12.Result, B=Multiply(A=ScalarParameter() {defaultValue=130, parameterName=Test}, B=@n16))
  n18 = ParticleColor()
  MaterialFunctionCall(Emissive=Multiply(A=Add(A=@n17), B=@n18.RGB)) {materialFunction=&MF_FX_EmissiveAdjustments}
  n19 = Power(Base=@n15.R, Exp=ScalarParameter() {defaultValue=0.3, parameterName=TexAToGlowPower})
  n20 = Multiply(A=@n2.R) {constB=2}
  n21 = Frac(None=@n20)
  n22 = LinearInterpolate(A=@n21, B=OneMinus(None=@n21), Alpha=Floor(None=@n20))
  n23 = ScalarParameter() {defaultValue=2, parameterName=GlowThickPower}
  n24 = Multiply(A=@n5, B=Multiply(A=Power(Base=@n22, Exp=@n23), B=ScalarParameter() {defaultValue=0.1, parameterName=GlowStrength}))
  Multiply(A=@n18.A, B=Add(A=Multiply(A=@n16, B=@n12.Result), B=Multiply(A=@n19, B=@n24)))
 ]
}
