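# Niagara system for the FirePetal wandering boss's bombard AOE burst, with shockwave, flash, light and decal layers.
# System MANAGE scripts supply spawn timing, world position and array index, replacing disabled one-shot bursts with gated per-frame spawns.
# Merged emitters select each layer's look by spawn group through MergeHelper switches.
# System parameters drive size, User.Color Variant blends palettes, camera distance scales effects, and Mobile and Switch skip some layers.
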NiagaraSystem /FirePetal_WanderingBoss/VFX/NS_FirePetal_WanderingBoss_Bombard_AOE_Burst_MergedV
dependencies {
 usedDataInterfaces [
  &NiagaraDataInterfacePlatformSet
  &NiagaraDataInterfaceCurve
  &NiagaraDataInterfaceCamera
  &NiagaraDataInterfaceColorCurve
  &NiagaraDataInterfaceArrayFloat
  &NiagaraDataInterfaceVector2DCurve
  &NiagaraDataInterfaceCollisionQuery
  &NiagaraDataInterfaceLandscape
  &NiagaraDataInterfaceArrayColor
 ]
 usedDynamicInputs [
  &ReturnNormalizedExecIndex
  &Power
  &Vector4FromCurve
  &Add_Float
  &SelectLinearColorFromArray
 ]
 usedModules [
  &self:SystemSpawnScript.SystemScriptSource.SystemScriptGraph.NiagaraNodeAssignment_0.SetVariables_#1
  &self:SystemSpawnScript.SystemScriptSource.SystemScriptGraph.NiagaraNodeAssignment_2.SetVariables_#2
  &SystemState
  &self:SystemSpawnScript.SystemScriptSource.SystemScriptGraph.NiagaraNodeAssignment_1.SetVariables_#3
  &self:MANAGE_ShouldSpawnThisFrame
  &self:MANAGE_UpdateCounter
 ]
}
emitters {
 DecalShockWave {
  emitterUpdateScript [
   &EmitterState()
   &SpawnBurst_Instantaneous(
    Spawn Probability=&MakeCustomFloatFromBool(Bool=$User.WaterFX, True Float=0, False Float=1)
    Spawn Time=0.1
    Use Spawn Probability=true
    UseLoopCountLimit=false
   ) {enabled=false}
   &SpawnBurst_Instantaneous(Spawn Group=1, Use Spawn Probability=true, UseLoopCountLimit=false) {enabled=false, module=SpawnBurst_Instantaneous001}
   &SpawnBurst_Instantaneous(Spawn Group=2, Use Spawn Probability=false, UseLoopCountLimit=false) {enabled=false, module=SpawnBurst_Instantaneous002}
   &SpawnBurst_Instantaneous(Spawn Group=3, Use Spawn Probability=false, UseLoopCountLimit=false) {enabled=false, module=SpawnBurst_Instantaneous003}
   &SpawnBurst_Instantaneous(Spawn Group=4, Spawn Time=0.15, Use Spawn Probability=false, UseLoopCountLimit=false) {enabled=false, module=SpawnBurst_Instantaneous004}
   &SpawnBurst_Instantaneous(Spawn Group=5, Use Spawn Probability=false, UseLoopCountLimit=false) {enabled=false, module=SpawnBurst_Instantaneous005}
   &SpawnBurst_Instantaneous(Spawn Count=1, Spawn Group=6, Use Spawn Probability=false) {enabled=false, module=SpawnBurst_Instantaneous006}
   &SpawnBurst_Instantaneous(Spawn Count=1, Spawn Group=7, Use Spawn Probability=false) {enabled=false, module=SpawnBurst_Instantaneous007}
   &SpawnBurst_Instantaneous(Spawn Count=1, Spawn Group=8, Use Spawn Probability=false) {enabled=false, module=SpawnBurst_Instantaneous008}
   &SpawnBurst_Instantaneous(Spawn Count=1, Spawn Group=9, Spawn Time=0.2, Use Spawn Probability=false) {enabled=false, module=SpawnBurst_Instantaneous009}
   &SpawnPerFrame(
    Spawn:NiagaraBool=$System.ShouldSpawn
    Spawn Count:NiagaraInt32=&MakeCustomIntFromBool(
     Bool=&IsPlatformSetActive(New Platform Set=DataInterface() {propertyValues {Platforms {
         qualityLevelMask=-32
         deviceProfileStates [
          {profileName=Mobile, qualityLevelMask=4294967295, setQualityLevelMask=4294967295}
          {profileName=Switch, qualityLevelMask=4294967295, setQualityLevelMask=4294967295}
         ]
        }}})
     True INT=0
     False INT=1
    )
    Use Spawn Probability=false
   )
   &SpawnPerFrame(Spawn:NiagaraBool=$System.ShouldSpawn, Spawn Group=1, Use Spawn Probability=false) {module=SpawnPerFrame001}
   &SpawnPerFrame(
    Spawn:NiagaraBool=$System.ShouldSpawn
    Spawn Count:NiagaraInt32=&MakeCustomIntFromBool(
     Bool=&IsPlatformSetActive(New Platform Set=DataInterface() {propertyValues {Platforms {
         qualityLevelMask=-32
         deviceProfileStates [
          {profileName=Mobile, qualityLevelMask=4294967295, setQualityLevelMask=4294967295}
          {profileName=Switch, qualityLevelMask=4294967295, setQualityLevelMask=4294967295}
         ]
        }}})
     True INT=0
     False INT=1
    )
    Spawn Group=2
    Use Spawn Probability=false
   ) {module=SpawnPerFrame002}
   &SpawnPerFrame(Spawn:NiagaraBool=$System.ShouldSpawn, Spawn Group=3, Use Spawn Probability=false) {module=SpawnPerFrame003}
   &SpawnPerFrame(
    Spawn:NiagaraBool=$System.ShouldSpawn
    Spawn Count:NiagaraInt32=&MakeCustomIntFromBool(
     Bool=&IsPlatformSetActive(New Platform Set=DataInterface() {propertyValues {Platforms {
         qualityLevelMask=-32
         deviceProfileStates [
          {profileName=Mobile, qualityLevelMask=4294967295, setQualityLevelMask=4294967295}
          {profileName=Switch, qualityLevelMask=4294967295, setQualityLevelMask=4294967295}
         ]
        }}})
     True INT=0
     False INT=1
    )
    Spawn Group=4
    Use Spawn Probability=false
   ) {module=SpawnPerFrame004}
   &SpawnPerFrame(
    Spawn:NiagaraBool=$System.ShouldSpawn
    Spawn Count:NiagaraInt32=&MakeCustomIntFromBool(
     Bool=&IsPlatformSetActive(New Platform Set=DataInterface() {propertyValues {Platforms {
         qualityLevelMask=-32
         deviceProfileStates [
          {profileName=Mobile, qualityLevelMask=4294967295, setQualityLevelMask=4294967295}
          {profileName=Switch, qualityLevelMask=4294967295, setQualityLevelMask=4294967295}
         ]
        }}})
     True INT=0
     False INT=1
    )
    Spawn Group=5
    Use Spawn Probability=false
   ) {module=SpawnPerFrame005}
   &SpawnPerFrame(
    Spawn:NiagaraBool=$System.ShouldSpawn
    Spawn Count:NiagaraInt32=&MakeCustomIntFromBool(
     Bool=&IsPlatformSetActive(New Platform Set=DataInterface() {propertyValues {Platforms {
         qualityLevelMask=-32
         deviceProfileStates [
          {profileName=Mobile, qualityLevelMask=4294967295, setQualityLevelMask=4294967295}
          {profileName=Switch, qualityLevelMask=4294967295, setQualityLevelMask=4294967295}
         ]
        }}})
     True INT=0
     False INT=1
    )
    Spawn Group=6
    Use Spawn Probability=false
   ) {module=SpawnPerFrame006}
   &SpawnPerFrame(
    Spawn:NiagaraBool=$System.ShouldSpawn
    Spawn Count:NiagaraInt32=&MakeCustomIntFromBool(
     Bool=&IsPlatformSetActive(New Platform Set=DataInterface() {propertyValues {Platforms {
         qualityLevelMask=-32
         deviceProfileStates [
          {profileName=Mobile, qualityLevelMask=4294967295, setQualityLevelMask=4294967295}
          {profileName=Switch, qualityLevelMask=4294967295, setQualityLevelMask=4294967295}
         ]
        }}})
     True INT=0
     False INT=1
    )
    Spawn Group=7
    Use Spawn Probability=false
   ) {module=SpawnPerFrame007}
   &SpawnPerFrame(Spawn:NiagaraBool=$System.ShouldSpawn, Spawn Group=8, Use Spawn Probability=false) {module=SpawnPerFrame008}
   &SpawnPerFrame(
    Spawn:NiagaraBool=$System.ShouldSpawn
    Spawn Count:NiagaraInt32=&MakeCustomIntFromBool(
     Bool=&IsPlatformSetActive(New Platform Set=DataInterface() {propertyValues {Platforms {
         qualityLevelMask=-32
         deviceProfileStates [
          {profileName=Mobile, qualityLevelMask=4294967295, setQualityLevelMask=4294967295}
          {profileName=Switch, qualityLevelMask=4294967295, setQualityLevelMask=4294967295}
         ]
        }}})
     True INT=0
     False INT=1
    )
    Spawn Group=9
    Use Spawn Probability=false
   ) {module=SpawnPerFrame009}
  ]
  particleSpawnScript [
   &self:MANAGE_InitialManageVariables(Position:NiagaraPosition=$System.SpawnPosition, SpawnArrayIndex:NiagaraInt32=$System.SpawnArrayIndex) {module=MANAGE_InitialManageVariables}
   &InitializeParticle(
    Color=&LerpLinearColors(
     StartColor=&MergeHelper_ColorSwitch(
      0=<rgba 0.7 0.0903 0.0098 1>
      1=<rgba 20 1.6667 0 3>
      2=<rgba 1 1 1 1>
      3=<rgba 50 0.8333 0 1>
      4=<rgba 1 1 1 1>
      5=<rgba 2 0.2863 0.06 1>
      6=<rgba 1 1 1 1>
      7=<rgba 0.2 0.0267 0 0.6>
      8=<rgba 0.02 0.0017 0 0.7>
      9=<rgba 5 0.4167 0 0.1>
      Index=&SpawnGroup()
      Num=::NewEnumerator8
     )
     EndColor=&MergeHelper_ColorSwitch(
      0=<rgba 0.0359 0.7 0.0151 1>
      1=<rgba 0.1147 20 0.3997 3>
      2=<rgba 0 1 0.5318 1>
      3=<rgba 2.1333 50 4.5607 1>
      4=<rgba 0.3618 1 0.2094 1>
      5=<rgba 0.3611 2 0.0847 1>
      6=<rgba 0.0498 1 0.0375 1>
      7=<rgba 0.0139 0.2 0 0.6>
      8=<rgba 0 0.02 0.0045 0.7>
      9=<rgba 1.5518 5 0.2284 0.1>
      Index=&SpawnGroup()
      Num=::NewEnumerator8
     )
     LerpFactor=&MakeFloatFromInt(Int=$User.Color Variant)
    )
    Lifetime=&MergeHelper_FloatSwitch(
     0=0.6
     1=0.9
     2=0.7
     3=0.6
     4=0.6
     5=0.25
     6=0.8072
     7=0.6
     8=1
     9=0.3
     Index=&SpawnGroup()
     Num=::NewEnumerator8
    )
    Mesh Scale=&MergeHelper_VectorSwitch(
     0=&VectorFromFloat(Value=&Multiply_Float(A=0.7696, B=$System.P_FXRadius))
     1=<0 0 0>
     2=&VectorFromFloat(Value=91.1571)
     3=&Multiply_VectorByFloat(Vector=<1.3 1.3 1.8>, Float=&Multiply_Float(A=4.034, B=$System.P_FXRadius))
     4=&VectorFromFloat(Value=&Multiply_Float(A=2, B=$System.P_FXRadius))
     5=<0 0 0>
     6=<0 0 0>
     7=<0 0 0>
     8=<0 0 0>
     9=<0 0 0>
     Index=&SpawnGroup()
     Num=::NewEnumerator8
    )
    Mesh Scale Mode=::NewEnumerator4
    Position=&TransformPosition(Position=$Particles.SpawnPosition, Source Space=::World, Destination Space=::Simulation)
    Position Mode=::NewEnumerator1
    Position Offset=&MergeHelper_VectorSwitch(
     0=<0 0 0>
     1=<0 0 0>
     2=<0 0 0>
     3=<0 0 -40>
     4=&MakeVector(X=0, Y=0, Z=50)
     5=<0 0 0>
     6=<0 0 20>
     7=<0 0 0>
     8=<0 0 0>
     9=<0 0 0>
     Index=&SpawnGroup()
     Num=::NewEnumerator8
    )
    Sprite Size=&MergeHelper_Vector2Switch(
     0=<10 10>
     1=&Multiply_Vector2D(A=<600 2550.9253>, B=&MakeVector2D(X=$System.P_FXRadius, Y=$System.P_FXHeight))
     2=<0 0>
     3=<0 0>
     4=<0 0>
     5=&Vector2DFromFloat(Value=953.4699)
     6=&Vector2DFromFloat(Value=&Multiply_Float(
       A=&Multiply_Float(A=552.9227, B=$System.P_FXRadius)
       B=&FN_PickFloatDependOnPlatform(TargetPlatformActive=1, TargetPlatformNoActive=0.7)
      ))
     7=&Multiply_Vector2D_ByFloat(Float=$System.P_FXRadius, Vector2D=<5000 5000>)
     8=<2000 1250>
     9=<111.8121 5351.791>
     Index=&SpawnGroup()
     Num=::NewEnumerator8
    )
    Sprite Size Mode=::NewEnumerator4
    UsePositionOffset=true
    Write Material Random=true
   )
   &InitialMeshOrientation(
    Mesh Axis To Orient=&MergeHelper_VectorSwitch(
     0=<1 0 0>
     1=<1 0 0>
     2=<1 0 0>
     3=<0 1 0>
     4=<0 0 1>
     5=<0 0 0>
     6=<0 0 0>
     7=<0 0 0>
     8=<0 0 0>
     9=<0 0 0>
     Index=&SpawnGroup()
     Num=::NewEnumerator8
    )
    Mesh Orientation Mode=::NewEnumerator2
    Orientation Vector=&MergeHelper_VectorSwitch(
     0=<1 0 0>
     1=<1 0 0>
     2=<1 0 0>
     3=<1 0 0>
     4=<0 0 1>
     5=<0 0 0>
     6=<0 0 1>
     7=<0 0 0>
     8=<0 0 0>
     9=<0 0 0>
     Index=&SpawnGroup()
     Num=::NewEnumerator8
    )
    Rotation=&MergeHelper_VectorSwitch(
     0=&MakeVector(X=0, Y=0, Z=&RandomRangeFloat(Minimum=0, Maximum=1))
     1=<0 0 0>
     2=<0 0 0>
     3=&RandomRangeVector(Minimum=<0 0 0>, Maximum=<0 0 1>)
     4=&MakeVector(X=0, Y=0, Z=&RandomRangeFloat(Minimum=0, Maximum=1))
     5=<0 0 0>
     6=<0 0 0>
     7=<0 0 0>
     8=<0 0 0>
     9=<0 0 0>
     Index=&SpawnGroup()
     Num=::NewEnumerator8
    )
    Use Rotation Vector=true
   )
   /SetVariables(Particles.VisibilityTag=&SpawnGroup()) {isSetParameters=true}
  ]
  particleUpdateScript [
   &self:MANAGE_ParticleDelay(DelayTime:NiagaraFloat=&MergeHelper_FloatSwitch(0=0.1, 1=0, 2=0, 3=0, 4=0.15, 5=0, 6=0, 7=0, 8=0, 9=0.2, Index=&SpawnGroup(), Num=::NewEnumerator8)) {module=MANAGE_ParticleDelay}
   &ParticleState()
   &Timeline(
    Initial Start Time:NiagaraFloat=&RandomRangeFloat(Minimum=0, Maximum=200)
    Play Rate=&MergeHelper_FloatSwitch(
     0=0
     1=&FloatFromCurve(FloatCurve=DataInterface() {propertyValues {Curve {keys [{interpMode=RCIM_Cubic, tangentMode=RCTM_Break, time=0, value=10}, {arriveTangent=-11.7536, tangentMode=RCTM_User, time=1, value=0}]}}})
     2=&FloatFromCurve(FloatCurve=DataInterface() {propertyValues {Curve {keys [{tangentMode=RCTM_User, time=0, value=6}, {tangentMode=RCTM_User, time=1, value=2}]}}})
     3=0
     4=0
     Index=&SpawnGroup()
     Num=::NewEnumerator3
    )
    Timeline End=100000
   )
   &ScaleSpriteSize(
    Scale Factor=&MergeHelper_Vector2Switch(
     0=<1 1>
     1=&Vector2DFromFloat(Value=&FloatFromCurve(
       FloatCurve=DataInterface() {propertyValues {Curve {keys [
           {arriveTangent=2.8056, interpMode=RCIM_Cubic, leaveTangent=2.8056, tangentMode=RCTM_User, time=0, value=0}
           {arriveTangent=0.6909, interpMode=RCIM_Cubic, leaveTangent=0.6909, tangentMode=RCTM_User, time=0.3, value=0.8}
           <at 1 1.2 RCIM_Cubic>
          ]}}}
       Scale Curve=0.8
      ))
     2=<0 0>
     3=<0 0>
     4=<0 0>
     5=&Vector2DFromFloat(Value=&FloatFromCurve(FloatCurve=DataInterface() {propertyValues {Curve {keys [
           {interpMode=RCIM_Cubic, leaveTangent=0.4211, tangentMode=RCTM_Break, time=0, value=0.3}
           {arriveTangent=0.9347, interpMode=RCIM_Cubic, tangentMode=RCTM_Break, time=1, value=1}
          ]}}}))
     6=&Vector2DFromFloat(Value=&FloatFromCurve(FloatCurve=DataInterface() {propertyValues {Curve {keys [
           {tangentMode=RCTM_User, time=0, value=0.5}
           {arriveTangent=0.5, interpMode=RCIM_Cubic, leaveTangent=0.5, time=0.3276, value=0.8147}
           {tangentMode=RCTM_User, time=1, value=1}
          ]}}}))
     7=&Vector2DFromFloat(Value=&FloatFromCurve(
       FloatCurve=DataInterface() {propertyValues {Curve {keys [
           {arriveTangent=4.1785, interpMode=RCIM_Cubic, leaveTangent=4.1785, tangentMode=RCTM_User, time=0, value=0}
           {arriveTangent=-0.0569, interpMode=RCIM_Cubic, leaveTangent=-0.0569, tangentMode=RCTM_User, time=0.15, value=1}
           {arriveTangent=-0.0405, interpMode=RCIM_Cubic, leaveTangent=-0.0405, tangentMode=RCTM_User, time=1, value=1}
          ]}}}
       Scale Curve=/Game/Effects/Niagara/Modules/Math/FN_CameraDistance(
        0Dist:NiagaraFloat=600
        1Dist:NiagaraFloat=1000
        Camera Query:NiagaraDataInterfaceCamera=DataInterface()
        Position:NiagaraPosition=$Particles.Position
       )
      ))
     8=&Vector2DFromFloat(Value=&FloatFromCurve(
       FloatCurve=DataInterface() {propertyValues {Curve {keys [
           {arriveTangent=-0.0569, interpMode=RCIM_Cubic, leaveTangent=-0.0569, tangentMode=RCTM_User, time=0.15, value=1}
           {arriveTangent=-0.0405, interpMode=RCIM_Cubic, leaveTangent=-0.0405, tangentMode=RCTM_User, time=1, value=1}
          ]}}}
       Scale Curve=&Multiply_Float(
        A=0.536
        B=&Multiply_Float(
         A=$System.P_FXRadius
         B=&FN_CameraDistance(
          0Dist:NiagaraFloat=600
          1Dist:NiagaraFloat=1000
          Camera Query:NiagaraDataInterfaceCamera=DataInterface()
          Position:NiagaraPosition=$Particles.Position
         )
        )
       )
      ))
     9=&MakeVector2D(
      X=&Multiply_Float(
       A=$System.P_FXRadius
       B=&FloatFromCurve(
        FloatCurve=DataInterface() {propertyValues {Curve {keys [
            {arriveTangent=-0.0031, interpMode=RCIM_Cubic, leaveTangent=-0.0031, tangentMode=RCTM_User, time=0.15, value=1}
            {arriveTangent=-0.001, interpMode=RCIM_Cubic, leaveTangent=-0.001, tangentMode=RCTM_User, time=1, value=1}
           ]}}}
        Scale Curve=0.536
       )
      )
      Y=&FloatFromCurve(
       FloatCurve=DataInterface() {propertyValues {Curve {keys [
           {arriveTangent=3.2187, interpMode=RCIM_Cubic, leaveTangent=3.2187, tangentMode=RCTM_User, time=0, value=0}
           {arriveTangent=0.0224, interpMode=RCIM_Cubic, leaveTangent=0.0224, tangentMode=RCTM_User, time=1, value=1}
          ]}}}
       Scale Curve=&Multiply_Float(
        A=$System.P_FXHeight
        B=&FloatFromCurve(
         FloatCurve=DataInterface() {propertyValues {Curve {keys [
             {arriveTangent=-0.0031, interpMode=RCIM_Cubic, leaveTangent=-0.0031, tangentMode=RCTM_User, time=0.15, value=1}
             {arriveTangent=-0.001, interpMode=RCIM_Cubic, leaveTangent=-0.001, tangentMode=RCTM_User, time=1, value=1}
            ]}}}
         Scale Curve=0.536
        )
       )
      )
     )
     Index=&SpawnGroup()
     Num=::NewEnumerator8
    )
    Scale Sprite Size Mode=::NewEnumerator2
   )
   &ScaleMeshSize(Scale Factor=&MergeHelper_VectorSwitch(
     0=&Multiply_Vector(
      A=&VectorFromFloat(Value=&FloatFromCurve(
        FloatCurve=DataInterface() {propertyValues {Curve {keys [
            {arriveTangent=4.1785, interpMode=RCIM_Cubic, leaveTangent=4.1785, tangentMode=RCTM_User, time=0, value=0.3}
            {arriveTangent=1.1179, interpMode=RCIM_Cubic, leaveTangent=1.1179, tangentMode=RCTM_User, time=0.2042, value=0.8}
            {arriveTangent=0.0121, interpMode=RCIM_Cubic, leaveTangent=0.0121, tangentMode=RCTM_User, time=0.8, value=1}
            {arriveTangent=0.0224, interpMode=RCIM_Cubic, leaveTangent=0.0224, tangentMode=RCTM_User, time=0.81, value=0}
           ]}}}
        Scale Curve=19
       ))
      B=$Engine.Owner.Scale
     )
     1=<1 1 1>
     2=&MakeVector(
      X=1.5
      Y=1.5
      Z=&FloatFromCurve(
       FloatCurve=DataInterface() {propertyValues {Curve {keys [{interpMode=RCIM_Cubic, leaveTangent=1.4308, tangentMode=RCTM_Break, time=0, value=0.5}, {tangentMode=RCTM_User, time=1, value=1.2}]}}}
       Scale Curve=2.6454
      )
     )
     3=&Multiply_VectorByFloat(
      Vector=<1 1 1>
      Float=&FloatFromCurve(FloatCurve=DataInterface() {propertyValues {Curve {keys [
           {arriveTangent=1.9083, interpMode=RCIM_Cubic, leaveTangent=1.9083, tangentMode=RCTM_User, time=0, value=0}
           {tangentMode=RCTM_User, time=1, value=1}
          ]}}})
     )
     4=&VectorFromFloat(Value=&Lerp_Float(
       A=0.3
       B=1.3
       Alpha=&FloatFromCurve(FloatCurve=DataInterface() {propertyValues {Curve {keys [
            {arriveTangent=4.1785, interpMode=RCIM_Cubic, leaveTangent=4.1785, tangentMode=RCTM_User, time=0, value=0}
            {arriveTangent=1.1179, interpMode=RCIM_Cubic, leaveTangent=1.1179, tangentMode=RCTM_User, time=0.2042, value=0.6228}
            {arriveTangent=0.0224, interpMode=RCIM_Cubic, leaveTangent=0.0224, tangentMode=RCTM_User, time=1, value=1}
           ]}}})
      ))
     5=<0 0 0>
     6=<0 0 0>
     7=<0 0 0>
     8=<0 0 0>
     9=<0 0 0>
     Index=&SpawnGroup()
     Num=::NewEnumerator8
    ))
   &ScaleColor(
    Scale Alpha=&MergeHelper_FloatSwitch(
     0=&Multiply_Float(
      A=&Lerp_Float(
       A=0.15
       B=1
       Alpha=&FN_CameraDistance(
        0Dist:NiagaraFloat=2048
        1Dist:NiagaraFloat=6000
        Camera Query:NiagaraDataInterfaceCamera=DataInterface() {propertyValues {bRequireCurrentFrameData=false}}
        Position:NiagaraPosition=$Particles.Position
       )
      )
      B=&FloatFromCurve(
       FloatCurve=DataInterface() {propertyValues {Curve {keys [
           {arriveTangent=9.0725, interpMode=RCIM_Cubic, leaveTangent=9.0725, tangentMode=RCTM_User, time=0, value=0}
           {arriveTangent=-2.5249, interpMode=RCIM_Cubic, leaveTangent=-2.5249, tangentMode=RCTM_User, time=0.0618, value=0.9847}
           {arriveTangent=-0.5077, interpMode=RCIM_Cubic, leaveTangent=-0.5077, tangentMode=RCTM_User, time=0.3179, value=0.1408}
           {arriveTangent=-0.0405, interpMode=RCIM_Cubic, leaveTangent=-0.0405, tangentMode=RCTM_User, time=1, value=0}
          ]}}}
       Scale Curve=&FNTimeOfDayScalar(Night Scale=0.4, Morning Scale=0.6, Day Scale=0.8, Evening Scale=0.6)
      )
     )
     1=&FloatFromCurve(FloatCurve=DataInterface() {propertyValues {Curve {keys [
          {arriveTangent=9.0725, interpMode=RCIM_Cubic, leaveTangent=9.0725, tangentMode=RCTM_User, time=0, value=0}
          {arriveTangent=-2.5249, interpMode=RCIM_Cubic, leaveTangent=-2.5249, tangentMode=RCTM_User, time=0.1408, value=0.9847}
          {arriveTangent=-1.146, interpMode=RCIM_Cubic, leaveTangent=-1.146, time=0.3, value=0.25}
          {arriveTangent=-0.0405, interpMode=RCIM_Cubic, leaveTangent=-0.0405, tangentMode=RCTM_User, time=1, value=0}
         ]}}})
     2=&FloatFromCurve(
      FloatCurve=DataInterface() {propertyValues {Curve {keys [
          {arriveTangent=9.0725, interpMode=RCIM_Cubic, leaveTangent=9.0725, tangentMode=RCTM_User, time=0, value=0}
          {arriveTangent=0.286, leaveTangent=0.286, time=0.1408, value=0.9847}
          {arriveTangent=-1.146, interpMode=RCIM_Cubic, leaveTangent=-1.146, tangentMode=RCTM_Break, time=0.7, value=1}
          {arriveTangent=-0.0405, interpMode=RCIM_Cubic, leaveTangent=-0.0405, tangentMode=RCTM_Break, time=1, value=0}
         ]}}}
      Scale Curve=&FN_CameraDistance(
       0Dist:NiagaraFloat=0
       1Dist:NiagaraFloat=800
       Camera Query:NiagaraDataInterfaceCamera=DataInterface()
       Position:NiagaraPosition=$Particles.Position
      )
     )
     3=&FloatFromCurve(FloatCurve=DataInterface() {propertyValues {Curve {keys [
          {arriveTangent=-1.9082, interpMode=RCIM_Cubic, leaveTangent=-1.9082, tangentMode=RCTM_User, time=0, value=2}
          {arriveTangent=-2, interpMode=RCIM_Cubic, leaveTangent=-2, time=0.2528, value=0.602}
          {interpMode=RCIM_Cubic, tangentMode=RCTM_User, time=1, value=0}
         ]}}})
     4=1
     5=&FloatFromCurve(FloatCurve=DataInterface() {propertyValues {Curve {keys [
          {tangentMode=RCTM_User, time=0, value=1}
          {arriveTangent=-0.074, interpMode=RCIM_Cubic, leaveTangent=-0.074, tangentMode=RCTM_User, time=0.3, value=1}
          {arriveTangent=-2.8624, interpMode=RCIM_Cubic, leaveTangent=-2.8624, tangentMode=RCTM_User, time=1, value=0}
         ]}}})
     6=0
     7=&FNTimeOfDayScalar(Night Scale=0.5, Morning Scale=1, Evening Scale=1)
     8=&FloatFromCurve(FloatCurve=DataInterface() {propertyValues {Curve {keys [
          {interpMode=RCIM_Cubic, leaveTangent=5.999, tangentMode=RCTM_Break, time=0, value=0}
          {interpMode=RCIM_Cubic, tangentMode=RCTM_Break, time=0.2, value=1}
          {arriveTangent=-2.7496, interpMode=RCIM_Cubic, leaveTangent=-2.7496, tangentMode=RCTM_User, time=1, value=0}
         ]}}})
     9=&FloatFromCurve(FloatCurve=DataInterface() {propertyValues {Curve {keys [
          {interpMode=RCIM_Cubic, tangentMode=RCTM_Break, time=0, value=1}
          {arriveTangent=0.1747, interpMode=RCIM_Cubic, leaveTangent=0.1747, tangentMode=RCTM_User, time=1, value=0}
         ]}}})
     Index=&SpawnGroup()
     Num=::NewEnumerator8
    )
    Scale RGB=&MergeHelper_VectorSwitch(
     0=&VectorFromFloat(Value=&FloatFromCurve(
       FloatCurve=DataInterface() {propertyValues {Curve {keys [
           <at 0 1 RCIM_Cubic>
           {arriveTangent=-1.4167, interpMode=RCIM_Cubic, leaveTangent=-1.4167, time=0.2, value=1}
           {arriveTangent=-1.25, interpMode=RCIM_Cubic, leaveTangent=-1.25, time=0.6, value=0.15}
           <at 1 0 RCIM_Cubic>
          ]}}}
       Scale Curve=5
      ))
     1=<1 1 1>
     2=&VectorFromFloat(Value=&FloatFromCurve(FloatCurve=DataInterface() {propertyValues {Curve {keys [{interpMode=RCIM_Cubic, tangentMode=RCTM_Break, time=0.3, value=1}, {arriveTangent=-0.14, tangentMode=RCTM_User, time=1, value=0}]}}}))
     3=&Multiply_VectorByFloat(
      Vector=&MakeVectorFromLinearColorRGB(Color=&ColorFromCurve(ColorCurve=DataInterface() {propertyValues {
          AlphaCurve {keys []}
          BlueCurve {keys [<at 0.1796 20 RCIM_Cubic>,<at 0.7522 0.01>]}
          GreenCurve {keys [<at 0.1796 7.5 RCIM_Cubic>,<at 0.7522 0.0039>]}
          RedCurve {keys [<at 0.1796 1.25 RCIM_Cubic>,<at 0.7522 0.0009>]}
         }}))
      Float=1
     )
     4=<1 1 1>
     5=<1 1 1>
     6=<0 0 0>
     7=&Multiply_VectorByFloat(
      Vector=<1 1 1>
      Float=&FloatFromCurve(FloatCurve=DataInterface() {propertyValues {Curve {keys [<at 0 1 RCIM_Cubic>, {interpMode=RCIM_Cubic, tangentMode=RCTM_Break, time=0.1, value=1}, <at 1 0>]}}})
     )
     8=<1 1 1>
     9=<1 1 1>
     Index=&SpawnGroup()
     Num=::NewEnumerator8
    )
    ScaleA=true
    ScaleRGB=true
   )
   &CameraOffset(Camera Offset Amount=&MergeHelper_FloatSwitch(
     0=0
     1=100
     2=&FN_LOD_Distance_Remap(Start Distance=100, End Distance=400, Minimum Output=0, Maximum Output=100, Clamp Output=true)
     3=20
     4=0
     5=&FN_LOD_Distance_Remap(Start Distance=100, End Distance=600, Minimum Output=0, Maximum Output=300, Clamp Output=true)
     6=&FN_LOD_Distance_Remap(Start Distance=100, End Distance=300, Minimum Output=0, Maximum Output=100, Clamp Output=true)
     7=&FN_LOD_Distance_Remap(Start Distance=200, End Distance=500, Minimum Output=0, Maximum Output=350, Clamp Output=true)
     8=&FN_LOD_Distance_Remap(Start Distance=300, End Distance=700, Minimum Output=0, Maximum Output=500, Clamp Output=true)
     9=&FN_LOD_Distance_Remap(Start Distance=300, End Distance=700, Minimum Output=0, Maximum Output=500, Clamp Output=true)
     Index=&SpawnGroup()
     Num=::NewEnumerator8
    ))
   &DynamicMaterialParameters(
    Index 0 Param 1=&MergeHelper_FloatSwitch(
     0=&FloatFromCurve(FloatCurve=DataInterface() {propertyValues {Curve {keys [
          {arriveTangent=4.1785, interpMode=RCIM_Cubic, leaveTangent=4.1785, tangentMode=RCTM_User, time=0, value=0}
          {arriveTangent=1.1179, interpMode=RCIM_Cubic, leaveTangent=1.1179, tangentMode=RCTM_User, time=0.2, value=0.6228}
          {arriveTangent=0.0224, interpMode=RCIM_Cubic, leaveTangent=0.0224, tangentMode=RCTM_User, time=1, value=1}
         ]}}})
     1=$Particles.Timeline.Time
     2=$Particles.Timeline.Time
     3=0
     4=&FloatFromCurve(FloatCurve=DataInterface() {propertyValues {Curve {keys [<at 0 0 RCIM_Cubic>, {arriveTangent=2.1767, interpMode=RCIM_Cubic, leaveTangent=2.1767, tangentMode=RCTM_User, time=1, value=1}]}}})
     5=0
     6=1
     7=0.2
     8=400
     9=400
     Index=&SpawnGroup()
     Num=::NewEnumerator8
    )
    Index 0 Param 2=&MergeHelper_FloatSwitch(
     0=&FloatFromCurve(FloatCurve=DataInterface() {propertyValues {Curve {keys [
          {interpMode=RCIM_Cubic, tangentMode=RCTM_User, time=0, value=1}
          {arriveTangent=-1, interpMode=RCIM_Cubic, leaveTangent=-1, time=0.5877, value=0.1326}
          {interpMode=RCIM_Cubic, tangentMode=RCTM_User, time=1, value=0}
         ]}}})
     1=0
     2=0
     3=0
     4=&Lerp_Float(
      A=4
      B=6
      Alpha=&FloatFromCurve(
       FloatCurve=DataInterface() {propertyValues {Curve {keys [
           {arriveTangent=-2.29, interpMode=RCIM_Cubic, leaveTangent=-2.29, tangentMode=RCTM_User, time=0, value=1}
           {interpMode=RCIM_Cubic, tangentMode=RCTM_User, time=1, value=0}
          ]}}}
       Scale Curve=5.1
      )
     )
     5=0
     6=1
     7=1
     8=1
     9=1
     Index=&SpawnGroup()
     Num=::NewEnumerator8
    )
    Index 0 Param 3=&MergeHelper_FloatSwitch(
     0=&Lerp_Float(
      A=0.7
      B=1
      Alpha=&FloatFromCurve(FloatCurve=DataInterface() {propertyValues {Curve {keys [<at 0 0 RCIM_Cubic>, {arriveTangent=2.7526, interpMode=RCIM_Cubic, leaveTangent=2.7526, tangentMode=RCTM_User, time=1, value=1}]}}})
     )
     1=0
     2=0
     3=0
     4=&Lerp_Float(
      A=0
      B=100
      Alpha=&FloatFromCurve(FloatCurve=DataInterface() {propertyValues {Curve {keys [
           {arriveTangent=-4.5741, interpMode=RCIM_Cubic, leaveTangent=-4.5741, tangentMode=RCTM_User, time=0, value=1}
           {interpMode=RCIM_Cubic, tangentMode=RCTM_User, time=0.5, value=0}
          ]}}})
     )
     5=0
     6=1
     7=0
     8=0
     9=0
     Index=&SpawnGroup()
     Num=::NewEnumerator8
    )
    Index 0 Param 4=&MergeHelper_FloatSwitch(0=0, 1=0, 2=0, 3=0, 4=0, 5=0, 6=1, 7=0, 8=0, 9=0, Index=&SpawnGroup(), Num=::NewEnumerator8)
    Module.Param0WriteEnabled=true
    Module.Param1WriteEnabled=true
    Module.Param2WriteEnabled=true
    Module.Param3WriteEnabled=true
   )
   &SpriteFacingAndAlignment(Sprite Alignment=<0 0 1>, WriteAlignment=true, WriteFacing=false)
   /SetVariables(Particles.Dy2:Vector4f=&MakeVector4(X=3, Y=100, Z=0, W=0)) {isSetParameters=true}
   &FN_Sprite_Attributes(
    Camera Offset=200
    Color=&MakeLinearColorFromVectorAndFloat(
     "Vector (RGB)"=&Multiply_Vector(A=&MakeVectorFromLinearColorRGB(Color=$System.color_e), B=<1 1 1>)
     "Float (Alpha)"=&Multiply_Float(
      A=&FloatFromCurve(FloatCurve=DataInterface() {propertyValues {Curve {keys [<at 0 1 RCIM_Cubic>, {arriveTangent=-15.1131, interpMode=RCIM_Cubic, leaveTangent=-15.1131, tangentMode=RCTM_User, time=0.2, value=0}]}}})
      B=&Multiply_Float(
       A=&FN_CameraDistance(
        0Dist:NiagaraFloat=200
        1Dist:NiagaraFloat=800
        Camera Query:NiagaraDataInterfaceCamera=DataInterface()
        Position:NiagaraPosition=$Particles.Position
       )
       B=1
      )
     )
    )
    Create Camera Offset Attribute=true
    Create Color Attribute=true
    Create SpriteSize Attribute=true
    SpriteSize=&Multiply_Vector2D_ByFloat(
     Float=&FloatFromCurve(
      FloatCurve=DataInterface() {propertyValues {Curve {keys [
          {interpMode=RCIM_Cubic, tangentMode=RCTM_User, time=0, value=1}
          {arriveTangent=-1.701, interpMode=RCIM_Cubic, leaveTangent=-1.701, tangentMode=RCTM_User, time=0.2, value=0.5}
          <at 0.3 0 RCIM_Cubic>
         ]}}}
      Scale Curve=&Multiply_Float(
       A=&FN_CameraDistance(
        0Dist:NiagaraFloat=200
        1Dist:NiagaraFloat=800
        Camera Query:NiagaraDataInterfaceCamera=DataInterface()
        Position:NiagaraPosition=$Particles.Position
       )
       B=2.3
      )
     )
     Vector2D=<500 500>
    )
    Write to Module Namespace=true
   )
  ]
  rendererClasses [&NiagaraMeshRendererProperties, &NiagaraSpriteRendererProperties]
  renderers [
   {rendererClass=&NiagaraMeshRendererProperties, rendererIndex=0}
   {rendererClass=&NiagaraSpriteRendererProperties, rendererIndex=1}
   {rendererClass=&NiagaraMeshRendererProperties, rendererIndex=2}
   {rendererClass=&NiagaraMeshRendererProperties, rendererIndex=3}
   {rendererClass=&NiagaraMeshRendererProperties, rendererIndex=4}
   {rendererClass=&NiagaraSpriteRendererProperties, rendererIndex=5}
   {rendererClass=&NiagaraSpriteRendererProperties, rendererIndex=6}
   {rendererClass=&NiagaraSpriteRendererProperties, rendererIndex=7}
   {rendererClass=&NiagaraSpriteRendererProperties, rendererIndex=8}
   {rendererClass=&NiagaraSpriteRendererProperties, rendererIndex=9}
  ]
 }
 LightAndDecal {
  emitterSpawnScript [
   /SetVariables(Emitter.CameraOffsetSprite=200) {isSetParameters=true}
   /SetVariables(Emitter.DecalColor:LinearColor=<rgba 0.4424 1 0.0908 1>) {isSetParameters=true, module=SetVariables_#4}
  ]
  emitterUpdateScript [
   &EmitterState()
   &SpawnBurst_Instantaneous(Use Spawn Probability=false, UseLoopCountLimit=false) {enabled=false}
   &SpawnBurst_Instantaneous(Use Spawn Probability=false, UseLoopCountLimit=true) {enabled=false, module=SpawnBurst_Instantaneous002}
   &SpawnBurst_Instantaneous(Spawn Group=1, Use Spawn Probability=false, UseLoopCountLimit=false) {enabled=false, module=SpawnBurst_Instantaneous001}
   &SpawnPerFrame(Spawn:NiagaraBool=$System.ShouldSpawn, Use Spawn Probability=false)
   &SpawnPerFrame(Spawn:NiagaraBool=$System.ShouldSpawn, Spawn Group=1, Use Spawn Probability=false) {module=SpawnPerFrame001}
  ]
  particleSpawnScript [
   &self:MANAGE_InitialManageVariables(Position:NiagaraPosition=$System.SpawnPosition, SpawnArrayIndex:NiagaraInt32=$System.SpawnArrayIndex) {module=MANAGE_InitialManageVariables}
   /SetVariables(Particles.VisibilityTag=$Emitter.SpawnGroup) {isSetParameters=true, module=SetVariables_#5}
   &InitializeParticle(
    Color=$System.color_d
    Lifetime=&Multiply_Float(
     A=&SelectFloatFromArray(Direct Array Index=$Particles.VisibilityTag, Float Selection Array=DataInterface() {propertyValues {FloatData [0.6,0.8]}})
     B=1
    )
    Position=&TransformPosition(Position=$Particles.SpawnPosition, Source Space=::World, Destination Space=::Simulation)
    Position Mode=::NewEnumerator1
    Sprite Rotation Mode=::NewEnumerator2
    Sprite Size Mode=::NewEnumerator3
    Uniform Sprite Size=&Multiply_Float(A=341.8826, B=$System.P_FXRadius)
    UsePositionOffset=false
    Write Material Random=true
   )
   /SetVariables(
    Particles.DecalFade=1
    Particles.DecalOrientation:Quat4f=&OrientMeshToVector(Look At Direction=<0 0 1>, Reference Axis Vector=<1 0 0>)
    Particles.DecalSize:Vector3f=<200 380 380>
   ) {enabled=false, isSetParameters=true, module=SetVariables_#6}
   &DynamicMaterialParameters(Module.Index0Param0Write=true, Module.Index0Param1Write=true, Module.Index0Param2Write=true, Module.Index0Param3Write=true)
  ]
  particleUpdateScript [
   &ParticleState()
   &CameraOffset(Camera Offset Amount=&FN_LOD_Distance_Remap(Start Distance=300, End Distance=700, Minimum Output=0, Maximum Output=500, Clamp Output=true))
   &ApplyOwnerScaleToAttributes(Scale Initial Sprite Size=true)
   &ScaleSpriteSize(
    Scale Sprite Size Mode=::NewEnumerator0
    Uniform Scale Factor=&Multiply_Float(
     A=2
     B=&Multiply_Float(
      A=&Multiply_Float(
       A=&FN_CameraDistance(
        0Dist:NiagaraFloat=200
        1Dist:NiagaraFloat=800
        Camera Query:NiagaraDataInterfaceCamera=DataInterface()
        Position:NiagaraPosition=$Particles.Position
       )
       B=1
      )
      B=1
     )
    )
   )
   &ScaleColor(
    Scale Alpha=&FloatFromCurve(FloatCurve=DataInterface() {propertyValues {Curve {keys [{interpMode=RCIM_Cubic, tangentMode=RCTM_User, time=0, value=1}, {interpMode=RCIM_Cubic, tangentMode=RCTM_User, time=1, value=0}]}}})
    Scale RGB=&VectorFromFloat(Value=&Lerp_Float(
      A=1
      B=3
      Alpha=&FloatFromCurve(FloatCurve=DataInterface() {propertyValues {Curve {keys [
           {arriveTangent=9.0725, interpMode=RCIM_Cubic, leaveTangent=9.0725, tangentMode=RCTM_User, time=0, value=0}
           {arriveTangent=-2.5249, interpMode=RCIM_Cubic, leaveTangent=-2.5249, tangentMode=RCTM_User, time=0.05, value=5}
           {arriveTangent=-1.146, leaveTangent=-1.146, time=0.2, value=0.1}
           {arriveTangent=-0.0405, interpMode=RCIM_Cubic, leaveTangent=-0.0405, tangentMode=RCTM_User, time=1, value=0}
          ]}}})
     ))
    ScaleA=true
    ScaleRGB=true
   ) {module=ScaleColor002}
   &FN_Light_Attributes(
    Color=&MultiplyLinearColor(Linear Color=$System.color_lit, Scale Factor=&MakeVector4_FromFloat(Float=20))
    Create Color Attribute=true
    Create Position Attribute=true
    Create Radius Attribute=true
    Enable Particle Light Enable=true
    Light Enabled=true
    Position=&AddVectorToPosition(Position=$Particles.Position, Vector=<0 0 200>)
    Radius=&Multiply_Float(
     A=&Multiply_Float(A=804.8151, B=1)
     B=&MakeFloatFromVector(Vector=$Engine.Owner.Scale, Channel=::NewEnumerator0)
    )
    Scale Color=&FloatFromCurve(
     FloatCurve=DataInterface() {propertyValues {Curve {keys [
         {interpMode=RCIM_Cubic, tangentMode=RCTM_User, time=0, value=1}
         {arriveTangent=-0.9162, interpMode=RCIM_Cubic, leaveTangent=-0.9162, tangentMode=RCTM_User, time=1, value=0}
        ]}}}
     Scale Curve=&Lerp_Float(
      A=5
      B=30
      Alpha=&FloatFromCurve(FloatCurve=DataInterface() {propertyValues {Curve {keys [
           {arriveTangent=9.0725, interpMode=RCIM_Cubic, leaveTangent=9.0725, tangentMode=RCTM_User, time=0, value=0}
           {arriveTangent=-2.5249, interpMode=RCIM_Cubic, leaveTangent=-2.5249, tangentMode=RCTM_User, time=0.05, value=0.9847}
           {arriveTangent=-1.146, leaveTangent=-1.146, time=0.2, value=0.02}
           {arriveTangent=-0.0405, interpMode=RCIM_Cubic, leaveTangent=-0.0405, tangentMode=RCTM_User, time=1, value=0}
          ]}}})
     )
    )
   )
   &SolveForcesAndVelocity(Clamp Velocity=false, Limit Acceleration=false)
   /SetVariables(
    Particles.DecalFade:NiagaraFloat=&Lerp_Float(
     A=0
     B=1
     Alpha=&FloatFromCurve(FloatCurve=DataInterface() {propertyValues {Curve {keys [
          {arriveTangent=12.3499, interpMode=RCIM_Cubic, leaveTangent=12.3499, tangentMode=RCTM_User, time=0, value=0}
          <at 0.2 1 RCIM_Cubic>
          {arriveTangent=-0.0037, interpMode=RCIM_Cubic, leaveTangent=-0.0037, tangentMode=RCTM_User, time=1, value=0}
         ]}}})
    )
    Particles.DecalSize:Vector3f=&Multiply_Vector(A=$Particles.Initial.DecalSize, B=$Engine.Owner.Scale)
   ) {enabled=false, isSetParameters=true, module=SetVariables_#7}
  ]
  rendererClasses [&NiagaraSpriteRendererProperties, &NiagaraLightRendererProperties]
  renderers [
   {rendererClass=&NiagaraSpriteRendererProperties}
   {rendererClass=&NiagaraLightRendererProperties, rendererIndex=1}
  ]
 }
 LightAndDecal001 {
  emitterSpawnScript [
   /SetVariables(Emitter.DecalColor:LinearColor=$System.color_decal) {isSetParameters=true, module=SetVariables_#4}
   /SetVariables(Emitter.CameraOffsetSprite=200) {isSetParameters=true}
  ]
  emitterUpdateScript [
   &EmitterState()
   &SpawnBurst_Instantaneous(
    Spawn Probability=&MakeCustomFloatFromBool(Bool=$User.SpawnDecal, True Float=1, False Float=0)
    Spawn Time=0.1
    Use Spawn Probability=true
    UseLoopCountLimit=false
   ) {enabled=false, module=SpawnBurst_Instantaneous001}
   &SpawnPerFrame(Spawn:NiagaraBool=$System.ShouldSpawn, Use Spawn Probability=false)
   &SpawnPerFrame(Spawn:NiagaraBool=$System.ShouldSpawn, Spawn Group=1, Use Spawn Probability=false) {module=SpawnPerFrame001}
   &SpawnBurst_Instantaneous(Spawn Group=1, Use Spawn Probability=false, UseLoopCountLimit=false) {enabled=false}
   &SpawnBurst_Instantaneous(Spawn Group=2, Use Spawn Probability=false, UseLoopCountLimit=false) {enabled=false, module=SpawnBurst_Instantaneous002}
  ]
  particleSpawnScript [
   &self:MANAGE_InitialManageVariables(Position:NiagaraPosition=$System.SpawnPosition, SpawnArrayIndex:NiagaraInt32=$System.SpawnArrayIndex) {module=MANAGE_InitialManageVariables}
   /SetVariables(Particles.VisibilityTag=&SpawnGroup()) {isSetParameters=true, module=SetVariables_#5}
   &InitializeParticle(
    Color=&LerpLinearColors(
     StartColor=&MergeHelper_ColorSwitch(
      0=<rgba 1 0.5772 0.0862 1>
      1=<rgba 1 1 1 1>
      2=<rgba 3 0.8625 0.15 2>
      3=<rgba 0 0 0 0>
      4=<rgba 0 0 0 0>
      Index=&SpawnGroup()
      Num=::NewEnumerator3
     )
     EndColor=&MergeHelper_ColorSwitch(
      0=<rgba 0.0506 1 0.1267 1>
      1=<rgba 1 1 1 1>
      2=<rgba 1.2883 3 0 2>
      3=<rgba 0 0 0 0>
      4=<rgba 0 0 0 0>
      Index=&SpawnGroup()
      Num=::NewEnumerator3
     )
     LerpFactor=&MakeFloatFromInt(Int=$User.Color Variant)
    )
    Lifetime=&MergeHelper_FloatSwitch(
     0=&Multiply_Float(
      A=&SelectFloatFromArray(Direct Array Index=$Particles.VisibilityTag, Float Selection Array=DataInterface() {propertyValues {FloatData [2,2]}})
      B=&MakeCustomFloatFromBool(Bool=$User.SpawnDecal, True Float=1, False Float=0.5)
     )
     1=0.6
     2=0.6
     3=0
     4=0
     Index=&SpawnGroup()
     Num=::NewEnumerator3
    )
    Mesh Scale Mode=::NewEnumerator3
    Mesh Uniform Scale=10
    Position=&TransformPosition(Position=$Particles.SpawnPosition, Source Space=::World, Destination Space=::Simulation)
    Position Mode=::NewEnumerator1
    Sprite Rotation Mode=::NewEnumerator2
    Sprite Size Mode=::NewEnumerator3
    Uniform Sprite Size=&MergeHelper_FloatSwitch(
     0=10
     1=10
     2=&Multiply_Float(A=340, B=$System.P_FXRadius)
     3=0
     4=0
     Index=&SpawnGroup()
     Num=::NewEnumerator3
    )
    UsePositionOffset=false
    Write Material Random=true
   )
   /SetVariables(
    Particles.DecalFade=1
    Particles.DecalOrientation:Quat4f=&OrientMeshToVector(Look At Direction=<0 0 1>, Reference Axis Vector=<1 0 0>)
    Particles.DecalSize:Vector3f=<200 500 500>
   ) {isSetParameters=true, module=SetVariables_#6}
  ]
  particleUpdateScript [
   &self:MANAGE_ParticleDelay(DelayTime:NiagaraFloat=&MergeHelper_FloatSwitch(0=0.1, 1=0, 2=0, 3=0, 4=0, Index=&SpawnGroup(), Num=::NewEnumerator3)) {module=MANAGE_ParticleDelay}
   &ParticleState()
   &DynamicMaterialParameters(
    Index 0 Param 1=&MakeFloatFromVector(Vector=&Multiply_Vector(A=$Particles.Scale, B=<1 1 1>), Channel=::NewEnumerator0)
    Index 0 Param 2=&MakeFloatFromVector(
     Vector=&TransformPosition(Position=$Particles.Position, Source Space=::Simulation, Destination Space=::World)
     Channel=::NewEnumerator0
    )
    Index 0 Param 3=&MakeFloatFromVector(
     Vector=&TransformPosition(Position=$Particles.Position, Source Space=::Simulation, Destination Space=::World)
     Channel=::NewEnumerator1
    )
    Index 0 Param 4=&MakeFloatFromVector(
     Vector=&TransformPosition(Position=$Particles.Position, Source Space=::Simulation, Destination Space=::World)
     Channel=::NewEnumerator2
    )
    Index 1 Param 1=&FloatFromCurve(FloatCurve=DataInterface() {propertyValues {Curve {keys [
         {interpMode=RCIM_Cubic, leaveTangent=-8.4786, tangentMode=RCTM_Break, time=0, value=2.6}
         {arriveTangent=-2.4925, interpMode=RCIM_Cubic, tangentMode=RCTM_Break, time=1, value=-1.7}
        ]}}})
    Index 1 Param 2=&FloatFromCurve(FloatCurve=DataInterface() {propertyValues {Curve {keys [
         {arriveTangent=4.7488, interpMode=RCIM_Cubic, leaveTangent=4.7488, tangentMode=RCTM_User, time=0.2, value=0.1}
         {arriveTangent=5.8572, interpMode=RCIM_Cubic, leaveTangent=5.8572, tangentMode=RCTM_User, time=1, value=2}
        ]}}})
    Index 1 Param 3=&FloatFromCurve(FloatCurve=DataInterface() {propertyValues {Curve {keys [
         {interpMode=RCIM_Cubic, leaveTangent=0.8368, tangentMode=RCTM_Break, time=0, value=0.3}
         {arriveTangent=5.6433, interpMode=RCIM_Cubic, tangentMode=RCTM_Break, time=1, value=3}
        ]}}})
    Index 1 Param 4=&FloatFromCurve(
     FloatCurve=DataInterface() {propertyValues {Curve {keys [
         {interpMode=RCIM_Cubic, leaveTangent=7.8774, tangentMode=RCTM_Break, time=0, value=0}
         {arriveTangent=-3.9574, interpMode=RCIM_Cubic, leaveTangent=-3.9574, tangentMode=RCTM_User, time=0.17, value=1}
         <at 0.7 0>
        ]}}}
     Scale Curve=0.4
    )
    Module.Index0Param0Write=true
    Module.Index0Param1Write=true
    Module.Index0Param2Write=true
    Module.Index0Param3Write=true
    Module.Index1Param0Write=true
    Module.Index1Param1Write=true
    Module.Index1Param2Write=true
    Module.Index1Param3Write=true
    Write Parameter Index 1=true
   ) {module=DynamicMaterialParameters001}
   /SetVariables(
    Particles.DecalFade:NiagaraFloat=&Lerp_Float(
     A=0
     B=1
     Alpha=&FloatFromCurve(FloatCurve=DataInterface() {propertyValues {Curve {keys [
          {arriveTangent=12.3499, interpMode=RCIM_Cubic, leaveTangent=12.3499, tangentMode=RCTM_User, time=0, value=0}
          <at 0.2 1 RCIM_Cubic>
          {arriveTangent=-0.0037, interpMode=RCIM_Cubic, leaveTangent=-0.0037, tangentMode=RCTM_User, time=1, value=0}
         ]}}})
    )
    Particles.DecalSize:Vector3f=&Multiply_VectorByFloat(Vector=&Multiply_Vector(A=$Particles.Initial.DecalSize, B=$Engine.Owner.Scale), Float=$System.P_FXRadius)
   ) {isSetParameters=true, module=SetVariables_#7}
   &ScaleColor(
    Scale Alpha=&MergeHelper_FloatSwitch(
     0=1
     1=&FloatFromCurve(
      FloatCurve=DataInterface() {propertyValues {Curve {keys [
          {arriveTangent=14.2646, interpMode=RCIM_Cubic, leaveTangent=14.2646, tangentMode=RCTM_User, time=0, value=0}
          {arriveTangent=-4.1895, interpMode=RCIM_Cubic, leaveTangent=-4.1895, tangentMode=RCTM_User, time=0.25, value=1.2}
          {arriveTangent=-3.3333, interpMode=RCIM_Cubic, leaveTangent=-3.3333, tangentMode=RCTM_User, time=1, value=0}
         ]}}}
      Scale Curve=0.75
     )
     2=&FloatFromCurve(FloatCurve=DataInterface() {propertyValues {Curve {keys [{interpMode=RCIM_Cubic, tangentMode=RCTM_User, time=0, value=1}, {interpMode=RCIM_Cubic, tangentMode=RCTM_User, time=1, value=0}]}}})
     3=0
     4=0
     Index=&SpawnGroup()
     Num=::NewEnumerator3
    )
    Scale RGB=&MergeHelper_VectorSwitch(
     0=<1 1 1>
     1=<1 1 1>
     2=&VectorFromFloat(Value=&Lerp_Float(
       A=1
       B=3
       Alpha=&FloatFromCurve(FloatCurve=DataInterface() {propertyValues {Curve {keys [
            {arriveTangent=9.0725, interpMode=RCIM_Cubic, leaveTangent=9.0725, tangentMode=RCTM_User, time=0, value=0}
            {arriveTangent=-2.5249, interpMode=RCIM_Cubic, leaveTangent=-2.5249, tangentMode=RCTM_User, time=0.05, value=5}
            {arriveTangent=-1.146, leaveTangent=-1.146, time=0.2, value=0.1}
            {arriveTangent=-0.0405, interpMode=RCIM_Cubic, leaveTangent=-0.0405, tangentMode=RCTM_User, time=1, value=0}
           ]}}})
      ))
     3=<0 0 0>
     4=<0 0 0>
     Index=0
     Num=::NewEnumerator3
    )
    ScaleA=true
    ScaleRGB=true
   )
   &CameraOffset(Camera Offset Amount=&MergeHelper_FloatSwitch(
     0=0
     1=0
     2=&FN_LOD_Distance_Remap(Start Distance=300, End Distance=700, Minimum Output=0, Maximum Output=500, Clamp Output=true)
     3=0
     4=0
     Index=&SpawnGroup()
     Num=::NewEnumerator3
    ))
   /SetVariables() {isSetParameters=true}
   &ScaleSpriteSize(
    Scale Sprite Size Mode=::NewEnumerator0
    Uniform Scale Factor=&MergeHelper_FloatSwitch(
     0=1
     1=1
     2=&Multiply_Float(A=2, B=&Multiply_Float(A=1, B=1))
     3=0
     4=0
     Index=&SpawnGroup()
     Num=::NewEnumerator3
    )
   )
   &FN_Light_Attributes(
    Color=&MultiplyLinearColor(Linear Color=<rgba 1 0.0833 0 1>, Scale Factor=&MakeVector4_FromFloat(Float=20))
    Create Color Attribute=true
    Create Position Attribute=true
    Create Radius Attribute=true
    Enable Particle Light Enable=true
    Light Enabled=true
    Position=&AddVectorToPosition(Position=$Particles.Position, Vector=<0 0 200>)
    Radius=&Multiply_Float(
     A=&Multiply_Float(A=804.8151, B=1)
     B=&MakeFloatFromVector(Vector=$Engine.Owner.Scale, Channel=::NewEnumerator0)
    )
    Scale Color=&FloatFromCurve(
     FloatCurve=DataInterface() {propertyValues {Curve {keys [
         {interpMode=RCIM_Cubic, tangentMode=RCTM_User, time=0, value=1}
         {arriveTangent=-0.9162, interpMode=RCIM_Cubic, leaveTangent=-0.9162, tangentMode=RCTM_User, time=1, value=0}
        ]}}}
     Scale Curve=&Lerp_Float(
      A=5
      B=30
      Alpha=&FloatFromCurve(FloatCurve=DataInterface() {propertyValues {Curve {keys [
           {arriveTangent=9.0725, interpMode=RCIM_Cubic, leaveTangent=9.0725, tangentMode=RCTM_User, time=0, value=0}
           {arriveTangent=-2.5249, interpMode=RCIM_Cubic, leaveTangent=-2.5249, tangentMode=RCTM_User, time=0.05, value=0.9847}
           {arriveTangent=-1.146, leaveTangent=-1.146, time=0.2, value=0.02}
           {arriveTangent=-0.0405, interpMode=RCIM_Cubic, leaveTangent=-0.0405, tangentMode=RCTM_User, time=1, value=0}
          ]}}})
     )
    )
   )
  ]
  rendererClasses [
   &NiagaraDecalRendererProperties
   &NiagaraMeshRendererProperties
   &NiagaraSpriteRendererProperties
   &NiagaraLightRendererProperties
  ]
  renderers [
   {rendererClass=&NiagaraDecalRendererProperties, rendererIndex=0}
   {rendererClass=&NiagaraMeshRendererProperties, rendererIndex=1}
   {rendererClass=&NiagaraSpriteRendererProperties, rendererIndex=2}
   {rendererClass=&NiagaraLightRendererProperties, rendererIndex=3}
  ]
 }
 Particle-Bright {
  simTarget=GPUComputeSim
  emitterUpdateScript [
   &EmitterState()
   &SpawnPerFrame(
    Spawn:NiagaraBool=$System.ShouldSpawn
    Spawn Count:NiagaraInt32=&MakeIntFromFloat(Float=&Multiply_Float(A=210, B=$Engine.Emitter.SpawnCountScale), Conversion Mode=::NewEnumerator1)
    Use Spawn Probability=false
   )
  ]
  particleSpawnScript [
   &self:MANAGE_InitialManageVariables_04(Position:NiagaraPosition=$System.SpawnPosition, SpawnArrayIndex:NiagaraInt32=$System.SpawnArrayIndex) {module=MANAGE_InitialManageVariables_04}
   &InitializeParticle(
    Color=$System.color_b
    Lifetime Min=0.5
    Lifetime Mode=::NewEnumerator1
    Position=&TransformPosition(Position=$Particles.SpawnPosition, Source Space=::World, Destination Space=::Simulation)
    Position Mode=::NewEnumerator1
    Sprite Size=&Multiply_Vector2D_ByFloat(Float=&RandomRangeFloat(Minimum=1, Maximum=2.5), Vector2D=<3 6>)
    Sprite Size Mode=::NewEnumerator4
    UsePositionOffset=false
    Write Material Random=true
   )
   &ShapeLocation(
    MaskSpawnGroup=false
    Sphere Distribution=::NewEnumerator1
    Sphere Radius=2
    U Position=&Divide_Float(
     A=/Niagara/DynamicInputs/Integer/V2/Round_Float(Float:NiagaraFloat=&Divide_Float(A=&RandomRangeFloat(Minimum=0, Maximum=100), B=10))
     B=10
    )
    UseOwnerScale=true
    V Position=&Divide_Float(
     A=&Round_Float(Float:NiagaraFloat=&Divide_Float(A=&RandomRangeFloat(Minimum=0, Maximum=100), B=33.33))
     B=3
    )
   ) {module=ShapeLocation001}
   &CameraOffset(Camera Offset Amount=30)
   &AddVelocity(
    Module.Use Falloff Distance=false
    Origin Offset=<0 0 -1.6561>
    Override Seed=false
    Velocity Mode=::NewEnumerator1
    Velocity Origin=&TransformPosition(Position=$Particles.SpawnPosition, Source Space=::World, Destination Space=::Simulation)
    Velocity Speed=&Multiply_Float(A=&RandomRangeFloat(Minimum=200, Maximum=1646.582), B=$System.P_FXRadius)
   )
  ]
  particleUpdateScript [
   &ParticleState()
   &ScaleSpriteSize(
    Scale Sprite Size Mode=::NewEnumerator0
    Uniform Scale Factor=&Multiply_Float(
     A=&FN_CameraDistance(
      0Dist:NiagaraFloat=50
      1Dist:NiagaraFloat=500
      Camera Query:NiagaraDataInterfaceCamera=DataInterface()
      Position:NiagaraPosition=$Particles.Position
     )
     B=1
    )
   )
   &CurlNoiseForce(
    Bias Noise=false
    Noise Frequency=5
    Noise Strength=&FloatFromCurve(
     FloatCurve=DataInterface() {propertyValues {Curve {keys [{interpMode=RCIM_Cubic, leaveTangent=3.7428, tangentMode=RCTM_Break, time=0.15, value=0}, <at 1 1>]}}}
     Scale Curve=3971.624
    )
    Pan Noise=true
   )
   &Drag(Drag=2.8463, Use Linear Drag=true, Use Rotational Drag=true)
   &GravityForce(Gravity=<0 0 600>)
   &SolveForcesAndVelocity(Clamp Velocity=false, Limit Acceleration=false)
   &ScaleColor(
    Scale Alpha=&FloatFromCurve(FloatCurve=DataInterface() {propertyValues {Curve {keys [
         {interpMode=RCIM_Cubic, tangentMode=RCTM_Break, time=0, value=0}
         <at 0.4 1>
         {arriveTangent=-0.0405, interpMode=RCIM_Cubic, leaveTangent=-0.0405, tangentMode=RCTM_User, time=0.7009, value=1}
         {arriveTangent=-8.0089, interpMode=RCIM_Cubic, leaveTangent=-8.0089, tangentMode=RCTM_User, time=1, value=0}
        ]}}})
    Scale RGB=&VectorFromFloat(Value=&FloatFromCurve(
      FloatCurve=DataInterface() {propertyValues {Curve {keys [
          {arriveTangent=-3.8472, interpMode=RCIM_Cubic, leaveTangent=-3.8472, tangentMode=RCTM_User, time=0, value=0}
          {arriveTangent=64.1147, interpMode=RCIM_Cubic, leaveTangent=-47.1467, tangentMode=RCTM_Break, time=0.3, value=10}
          {arriveTangent=-2.8456, interpMode=RCIM_Cubic, leaveTangent=-2.8456, tangentMode=RCTM_User, time=0.8, value=0}
          {arriveTangent=-5.983, tangentMode=RCTM_User, time=1, value=0}
         ]}}}
      Scale Curve=3
     ))
    ScaleA=true
    ScaleRGB=true
   )
   &FN_Sprite_Attributes(
    Color=&MultiplyLinearColor(
     Linear Color=$System.color_f
     Scale Factor=&MakeVector4(
      X=1
      Y=1
      Z=1
      W=&FloatFromCurve(FloatCurve=DataInterface() {propertyValues {Curve {keys [
           {interpMode=RCIM_Cubic, tangentMode=RCTM_Break, time=0, value=0}
           <at 0.4 1>
           {arriveTangent=-0.0405, interpMode=RCIM_Cubic, leaveTangent=-0.0405, tangentMode=RCTM_User, time=0.7009, value=1}
           {arriveTangent=-8.0089, interpMode=RCIM_Cubic, leaveTangent=-8.0089, tangentMode=RCTM_User, time=1, value=0}
          ]}}})
     )
    )
    Create Color Attribute=true
    Create SpriteSize Attribute=true
    SpriteSize=&Multiply_Vector2D_ByFloat(
     Float=&Multiply_Float(
      A=&FN_CameraDistance(
       0Dist:NiagaraFloat=50
       1Dist:NiagaraFloat=500
       Camera Query:NiagaraDataInterfaceCamera=DataInterface()
       Position:NiagaraPosition=$Particles.Position
      )
      B=1
     )
     Vector2D=<150 400>
    )
    Write to Module Namespace=true
   )
  ]
  rendererClasses [&NiagaraSpriteRendererProperties]
  renderers [
   {rendererClass=&NiagaraSpriteRendererProperties}
   {rendererClass=&NiagaraSpriteRendererProperties, rendererIndex=1}
  ]
 }
 Particle-Dissolve {
  simTarget=GPUComputeSim
  emitterUpdateScript [
   &EmitterState()
   &SpawnPerFrame(
    Spawn:NiagaraBool=$System.ShouldSpawn
    Spawn Count:NiagaraInt32=&MakeIntFromFloat(Float=&Multiply_Float(A=90, B=$Engine.Emitter.SpawnCountScale), Conversion Mode=::NewEnumerator1)
    Use Spawn Probability=false
   )
   &SpawnRate(SpawnRate=200, Use Spawn Probability=false) {enabled=false, module=SpawnRate001}
   &SpawnRate(
    SpawnRate=&FloatFromCurve(
     FloatCurve=DataInterface() {propertyValues {Curve {keys [
         {interpMode=RCIM_Cubic, tangentMode=RCTM_Break, time=0, value=0}
         {arriveTangent=0.7041, interpMode=RCIM_Cubic, leaveTangent=0.7041, tangentMode=RCTM_User, time=0.3, value=0.05}
         {arriveTangent=2.4035, tangentMode=RCTM_User, time=1, value=1}
        ]}}}
     CurveIndex=$Emitter.NormalizedLoopAge
     Scale Curve=376.5758
    )
    Use Spawn Probability=false
   ) {enabled=false}
  ]
  particleSpawnScript [
   &self:MANAGE_InitialManageVariables_05(Position:NiagaraPosition=$System.SpawnPosition, SpawnArrayIndex:NiagaraInt32=$System.SpawnArrayIndex) {module=MANAGE_InitialManageVariables_05}
   &InitializeParticle(
    Color=&LerpLinearColors(
     StartColor=&RandomRangeLinearColor(Minimum=<rgba 0.3 0.125 0.09 1>, Maximum=<rgba 0.4 0.1667 0.12 1>)
     EndColor=&RandomRangeLinearColor(Minimum=<rgba 0.2012 0.3 0.1139 1>, Maximum=<rgba 0.1539 0.4 0.2127 1>)
     LerpFactor=&MakeFloatFromInt(Int=$User.Color Variant)
    )
    Lifetime Max=1.7
    Lifetime Min=0.2
    Lifetime Mode=::NewEnumerator1
    Mesh Scale Mode=::NewEnumerator3
    Mesh Uniform Scale=100
    Position=&TransformPosition(Position=$Particles.SpawnPosition, Source Space=::World, Destination Space=::Simulation)
    Position Mode=::NewEnumerator1
    Sprite Rotation Mode=::NewEnumerator2
    Sprite Size Mode=::NewEnumerator1
    Uniform Sprite Size Max=90
    Uniform Sprite Size Min=30
    UsePositionOffset=false
    Write Material Random=true
   )
   /SetVariables() {isSetParameters=true}
   &AddVelocity(
    Module.Use Falloff Distance=false
    Override Seed=false
    Velocity Mode=::NewEnumerator1
    Velocity Origin=&TransformPosition(Position=$Particles.SpawnPosition, Source Space=::World, Destination Space=::Simulation)
    Velocity Speed=&Multiply_Float(
     A=&RandomRangeFloat(Minimum=1400, Maximum=1900)
     B=&FloatFromCurve(
      FloatCurve=DataInterface() {propertyValues {Curve {keys [
          {interpMode=RCIM_Cubic, leaveTangent=-1.3134, tangentMode=RCTM_Break, time=0.4, value=1}
          {arriveTangent=0.0485, tangentMode=RCTM_User, time=1, value=0.6}
         ]}}}
      CurveIndex=$Emitter.Age
     )
    )
   )
   &ShapeLocation(
    Cylinder Height=10
    MaskSpawnGroup=false
    Offset=<0 0 100>
    Offset Mode=::NewEnumerator0
    Override Seed=false
    Shape Primitive=::NewEnumerator1
    Surface Only=true
    Surface Only Band Thickness=1
    UseOwnerScale=true
   )
   &InitialMeshOrientation(Use Rotation Vector=false) {enabled=false}
   &CameraOffset(Camera Offset Amount=100)
  ]
  particleUpdateScript [
   &self:MANAGE_ParticleDelay_01(DelayTime=0.1) {module=MANAGE_ParticleDelay_01}
   &ParticleState()
   &ScaleSpriteSize(
    Scale Sprite Size Mode=::NewEnumerator0
    Uniform Scale Factor=&FN_LOD_Distance_Remap(Start Distance=200, End Distance=600, Minimum Output=0, Maximum Output=1, Clamp Output=true)
   ) {module=ScaleSpriteSize002}
   &ScaleSpriteSize(Scale Sprite Size Mode=::NewEnumerator2) {module=ScaleSpriteSize001}
   &ScaleSpriteSize(Scale Factor=&Vector2DFromFloat(Value=&FloatFromCurve(
      FloatCurve=DataInterface() {propertyValues {Curve {keys [
          {arriveTangent=4.0375, interpMode=RCIM_Cubic, leaveTangent=4.0375, tangentMode=RCTM_User, time=0, value=0}
          {arriveTangent=-0.0952, interpMode=RCIM_Cubic, leaveTangent=-0.0952, tangentMode=RCTM_User, time=0.2, value=0.4}
         ]}}}
      Scale Curve=$User.ElementSizeScale
     )))
   /Niagara/Modules/Update/SubUV/SubUVAnimation(Number Of Frames=36) {enabled=false}
   &GravityForce(Gravity=&RandomRangeVector(Minimum=<0 0 50>, Maximum=<0 0 1720.1442>))
   &WindForce(
    Wind Speed=&TransformVector(
     Vector=&RandomRangeVector(Minimum=<239.2935 0 0>, Maximum=<1422.8411 1 1>)
     Source Space=::World
     Destination Space=::Simulation
    )
    Wind Speed Scale=&FloatFromCurve(
     FloatCurve=DataInterface() {propertyValues {Curve {keys [{interpMode=RCIM_Cubic, tangentMode=RCTM_Break, time=0, value=0}, {arriveTangent=2.6532, tangentMode=RCTM_User, time=1, value=1}]}}}
     CurveIndex=$Emitter.Age
    )
   ) {enabled=false}
   &CurlNoiseForce(
    Bias Noise=false
    Noise Frequency=20
    Noise Strength=&FloatFromCurve(
     FloatCurve=DataInterface() {propertyValues {Curve {keys [{interpMode=RCIM_Cubic, tangentMode=RCTM_Break, time=0, value=1}, {arriveTangent=3.0586, tangentMode=RCTM_User, time=1, value=2}]}}}
     CurveIndex=$Emitter.Age
     Scale Curve=&RandomRangeFloat(Minimum=920.5424, Maximum=1328.9581)
    )
    Pan Noise=true
    Pan Noise Field=<0 0 2>
   )
   &Drag(Drag=&RandomRangeFloat(Minimum=3, Maximum=6), Use Linear Drag=true, Use Rotational Drag=true)
   &PointAttractionForce(
    Attraction Radius=3077.5405
    Attraction Strength=&FloatFromCurve(
     FloatCurve=DataInterface() {propertyValues {Curve {keys [{interpMode=RCIM_Cubic, leaveTangent=1.4775, tangentMode=RCTM_Break, time=0, value=0}, <at 1 1>]}}}
     Scale Curve=&RandomRangeFloat(Minimum=176.8817, Maximum=398.2867)
    )
    Attractor Position:NiagaraPosition=&TransformPosition(Position=$Particles.SpawnPosition, Source Space=::World, Destination Space=::Simulation)
    Attractor Position Offset:Vector3f=<0 0 263.7948>
    Kill Within Radius=false
    Module.Use Falloff=true
    UseAttractorOffset=true
   )
   &SolveForcesAndVelocity(Clamp Velocity=false, Limit Acceleration=false)
   &ScaleColor(
    Scale Alpha=&FloatFromCurve(FloatCurve=DataInterface() {propertyValues {Curve {keys [{interpMode=RCIM_Cubic, leaveTangent=-3.1003, tangentMode=RCTM_Break, time=0.2, value=1}, <at 1 0>]}}})
    ScaleA=true
    ScaleRGB=true
   )
   &DynamicMaterialParameters(
    Index 0 Param 1=&RandomRangeFloat(Minimum=0, Maximum=100)
    Index 0 Param 2=&FloatFromCurve(FloatCurve=DataInterface() {propertyValues {Curve {keys [{interpMode=RCIM_Cubic, leaveTangent=0.058, tangentMode=RCTM_Break, time=0, value=-1}, <at 1 1>]}}})
    Index 0 Param 3=&FloatFromCurve(FloatCurve=DataInterface() {propertyValues {Curve {keys [{interpMode=RCIM_Cubic, leaveTangent=-5.95, tangentMode=RCTM_Break, time=0.2, value=1}, <at 0.4 0>]}}})
    Module.Index0Param0Write=true
    Module.Index0Param1Write=true
    Module.Index0Param2Write=true
    Module.Index0Param3Write=true
   )
   &FN_Sprite_Attributes(
    Color=$System.color_g
    Create Color Attribute=true
    Create SpriteSize Attribute=true
    SpriteSize=<366.7501 353.7109>
    Write to Module Namespace=true
   )
  ]
  rendererClasses [&NiagaraSpriteRendererProperties]
  renderers [{rendererClass=&NiagaraSpriteRendererProperties}]
 }
 Smoke_Ring001 {
  emitterSpawnScript [/SetVariables(Emitter.R:NiagaraFloat=&RandomRangeFloat(Minimum=0, Maximum=1)) {isSetParameters=true}]
  emitterUpdateScript [
   &EmitterState()
   /SetVariables(Emitter.LightingCenter:Vector4f=&MakeFloat4FromVectorAndFloat(
     "Vector (RGB)"=&TransformPosition(Position=<0 0 0>, Source Space=::Local, Destination Space=::World)
     "Float (Alpha)"=1
    )) {enabled=false, isSetParameters=true}
   &SpawnPerFrame(
    Spawn:NiagaraBool=$System.ShouldSpawn
    Spawn Count:NiagaraInt32=&MakeIntFromFloat(Float=&Multiply_Float(A=15, B=$Engine.Emitter.SpawnCountScale), Conversion Mode=::NewEnumerator1)
    Use Spawn Probability=false
   )
   &SpawnBurst_Instantaneous(Spawn Count=15, Use Spawn Probability=false, UseLoopCountLimit=false) {enabled=false}
  ]
  particleSpawnScript [
   &self:MANAGE_InitialManageVariables_03(Position:NiagaraPosition=$System.SpawnPosition, SpawnArrayIndex:NiagaraInt32=$System.SpawnArrayIndex) {module=MANAGE_InitialManageVariables_03}
   &InitializeParticle(
    Color=$System.color_a
    Lifetime=&RandomRangeFloat(Minimum=0.6, Maximum=0.9)
    Position=&TransformPosition(Position=$Particles.SpawnPosition, Source Space=::World, Destination Space=::Simulation)
    Position Mode=::NewEnumerator1
    Position Offset=<0 0 -6>
    Sprite Rotation Mode=::NewEnumerator2
    Sprite Size Mode=::NewEnumerator3
    Uniform Sprite Size=&Multiply_Float(
     A=&RandomRangeFloat(Minimum=100, Maximum=130)
     B=&Multiply_Float(A=1.25, B=$System.P_FXRadius)
    )
    UsePositionOffset=true
    Write Material Random=true
   ) {module=InitializeParticle001}
   &CameraOffset(Camera Offset Amount=100)
   &ShapeLocation(
    Handle Radius=5
    Large Radius=15
    MaskSpawnGroup=false
    Non Uniform Scale=<1 1 0>
    Override Seed=false
    Shape Primitive=::NewEnumerator3
    UseOwnerScale=true
   )
   &AddVelocity(
    Module.Use Falloff Distance=false
    Origin Offset=&RandomRangeVector(Minimum=<0 0 -5>, Maximum=<0 0 -10>)
    Override Seed=false
    Velocity Mode=::NewEnumerator1
    Velocity Origin=&TransformPosition(Position=$Particles.SpawnPosition, Source Space=::World, Destination Space=::Simulation)
    Velocity Speed=&Multiply_Float(A=&RandomRangeFloat(Minimum=692.108, Maximum=916.4416), B=$System.P_FXRadius)
   )
  ]
  particleUpdateScript [
   &ParticleState()
   &ScaleSpriteSize(Scale Factor=<1.2 1>, Scale Sprite Size Mode=::NewEnumerator2) {module=ScaleSpriteSize001}
   &Drag(Drag=3, Use Linear Drag=true, Use Rotational Drag=false)
   &GravityForce(Gravity=<0 0 10>)
   &SolveForcesAndVelocity(Clamp Velocity=false, Limit Acceleration=false)
   &ScaleSpriteSize(Scale Factor=&Vector2DFromFloat(Value=&FloatFromCurve(
      FloatCurve=DataInterface() {propertyValues {Curve {keys [
          {interpMode=RCIM_Cubic, leaveTangent=1.7304, tangentMode=RCTM_Break, time=0, value=0.3}
          {arriveTangent=-0.0555, interpMode=RCIM_Cubic, leaveTangent=-0.0555, tangentMode=RCTM_User, time=0.5, value=1}
          <at 0.8 1>
         ]}}}
      Scale Curve=2
     )))
   /SetVariables(Particles.Color_ParticleEnergy:LinearColor=&ColorFromCurve(ColorCurve=DataInterface() {propertyValues {
       AlphaCurve {keys [<at 0 1 RCIM_Cubic>]}
       BlueCurve {keys [<at 0.1547 0>,<at 0.9643 0 RCIM_Cubic>]}
       GreenCurve {keys [<at 0.1547 0.1667>,<at 0.9643 0 RCIM_Cubic>]}
       RedCurve {keys [<at 0.1547 2>,<at 0.9643 0 RCIM_Cubic>]}
      }})) {isSetParameters=true}
   &DynamicMaterialParameters(
    Index 0 Param 1=&FloatFromCurve(FloatCurve=DataInterface() {propertyValues {Curve {keys [{tangentMode=RCTM_User, time=0, value=0}, {tangentMode=RCTM_User, time=1, value=1}]}}})
    Index 0 Param 3=&FloatFromCurve(
     FloatCurve=DataInterface() {propertyValues {Curve {keys [
         {interpMode=RCIM_Cubic, tangentMode=RCTM_Break, time=0, value=0}
         {interpMode=RCIM_Cubic, leaveTangent=-3.8707, tangentMode=RCTM_Break, time=0.2, value=1}
         {tangentMode=RCTM_User, time=1, value=0}
        ]}}}
     Scale Curve=10
    )
    Index 0 Param 4=&Multiply_Float(
     A=&Waveform(Global Phase=$Particles.Age, "Waveform [1]"=::NewEnumerator12, "[1] Phase"=$Particles.Age, "[1] Frequency"=1, "[1] Amplitude Min/Max"=<0 1>)
     B=&FloatFromCurve(FloatCurve=DataInterface() {propertyValues {Curve {keys [<at 0 20>,<at 0.7 1>]}}})
    )
    Index 1 Param 1=&MakeFloatFromVector(Vector=&MakeVectorFromLinearColorRGB(Color=$Particles.Color_ParticleEnergy), Channel=::NewEnumerator0)
    Index 1 Param 2=&MakeFloatFromVector(Vector=&MakeVectorFromLinearColorRGB(Color=$Particles.Color_ParticleEnergy), Channel=::NewEnumerator1)
    Index 1 Param 3=&MakeFloatFromVector(Vector=&MakeVectorFromLinearColorRGB(Color=$Particles.Color_ParticleEnergy), Channel=::NewEnumerator2)
    Index 1 Param 4=&FloatFromCurve(FloatCurve=DataInterface() {propertyValues {Curve {keys [
         {interpMode=RCIM_Cubic, leaveTangent=0.099, tangentMode=RCTM_Break, time=0, value=0}
         {arriveTangent=2.8269, tangentMode=RCTM_User, time=1, value=1.2}
        ]}}})
    Index 2 Param 1=284.3125
    Index 2 Param 2=&Multiply_Float(A=800, B=$System.P_FXRadius)
    Index 2 Param 3=&Multiply_Float(A=500, B=$System.P_FXRadius)
    Index 2 Param 4=&FloatFromCurve(FloatCurve=DataInterface() {propertyValues {Curve {keys [
         {interpMode=RCIM_Cubic, leaveTangent=0.8381, tangentMode=RCTM_Break, time=0.5, value=0}
         {arriveTangent=2.6267, interpMode=RCIM_Cubic, tangentMode=RCTM_Break, time=1, value=1}
        ]}}})
    Index 3 Param 1=1
    Index 3 Param 2=&MakeFloatFromVector(Vector=Linked() {linkedVariable {name=Particles.SpawnPosition, type=NiagaraPosition}}, Channel=::NewEnumerator0)
    Index 3 Param 3=&MakeFloatFromVector(Vector=Linked() {linkedVariable {name=Particles.SpawnPosition, type=NiagaraPosition}}, Channel=::NewEnumerator1)
    Index 3 Param 4=&MakeFloatFromVector(Vector=Linked() {linkedVariable {name=Particles.SpawnPosition, type=NiagaraPosition}}, Channel=::NewEnumerator2)
    Module.Index0Param0Write=true
    Module.Index0Param1Write=true
    Module.Index0Param2Write=true
    Module.Index0Param3Write=true
    Module.Index1Param0Write=true
    Module.Index1Param1Write=true
    Module.Index1Param2Write=true
    Module.Index1Param3Write=true
    Module.Index2Param0Write=true
    Module.Index2Param1Write=true
    Module.Index2Param2Write=true
    Module.Index2Param3Write=true
    Module.Index3Param0Write=true
    Module.Index3Param1Write=true
    Module.Index3Param2Write=true
    Module.Index3Param3Write=true
    Write Parameter Index 1=true
    Write Parameter Index 2=true
    Write Parameter Index 3=true
   )
   &ScaleColor(
    Scale Alpha=&FloatFromCurve(
     FloatCurve=DataInterface() {propertyValues {Curve {keys [
         {interpMode=RCIM_Cubic, leaveTangent=0.0758, tangentMode=RCTM_Break, time=0, value=0}
         {interpMode=RCIM_Cubic, tangentMode=RCTM_Break, time=0.35, value=1}
         {interpMode=RCIM_Cubic, leaveTangent=-0.0057, tangentMode=RCTM_Break, time=0.5, value=1}
         {arriveTangent=-4.4597, interpMode=RCIM_Cubic, leaveTangent=-4.4597, tangentMode=RCTM_User, time=1, value=0}
        ]}}}
     Scale Curve=&FN_CameraDistance(
      0Dist:NiagaraFloat=0
      1Dist:NiagaraFloat=500
      Camera Query:NiagaraDataInterfaceCamera=DataInterface()
      Position:NiagaraPosition=$Particles.Position
     )
    )
    ScaleA=true
    ScaleRGB=true
   )
  ]
  rendererClasses [&NiagaraSpriteRendererProperties]
  renderers [{rendererClass=&NiagaraSpriteRendererProperties}]
 }
 Smoke_Wisp001 {
  emitterUpdateScript [
   &EmitterState()
   &SpawnBurst_Instantaneous(Spawn Count=6, Use Spawn Probability=false, UseLoopCountLimit=false) {enabled=false}
   &SpawnPerFrame(
    Spawn:NiagaraBool=$System.ShouldSpawn
    Spawn Count:NiagaraInt32=&MakeIntFromFloat(Float=&Multiply_Float(A=6, B=$Engine.Emitter.SpawnCountScale), Conversion Mode=::NewEnumerator1)
    Use Spawn Probability=false
   )
   &SpawnPerFrame(
    Spawn:NiagaraBool=$System.ShouldSpawn
    Spawn Count:NiagaraInt32=&MakeIntFromFloat(Float=&Multiply_Float(A=4.8, B=$Engine.Emitter.SpawnCountScale), Conversion Mode=::NewEnumerator1)
    Spawn Group=1
    Use Spawn Probability=false
   ) {module=SpawnPerFrame001}
  ]
  particleSpawnScript [
   /SetVariables(Particles.VisibilityTag=&SpawnGroup()) {isSetParameters=true}
   &self:MANAGE_InitialManageVariables(Position:NiagaraPosition=$System.SpawnPosition, SpawnArrayIndex:NiagaraInt32=$System.SpawnArrayIndex) {module=MANAGE_InitialManageVariables}
   &InitializeParticle(
    Lifetime Max=1.2
    Lifetime Min=0.6
    Lifetime Mode=::NewEnumerator1
    Mesh Scale Mode=::NewEnumerator1
    Mesh Uniform Scale Max=1.25
    Position=&TransformPosition(Position=$Particles.SpawnPosition, Source Space=::World, Destination Space=::Simulation)
    Position Mode=::NewEnumerator1
    Sprite Rotation Mode=::NewEnumerator2
    Sprite Size Mode=::NewEnumerator3
    Uniform Sprite Size=&Multiply_Float(A=175, B=1.5266)
    UsePositionOffset=true
    Write Material Random=true
   ) {module=InitializeParticle001}
   /SetVariables(Particles.RandomFrame=&RandomRangeInt(Minimum=0, Maximum=63)) {isSetParameters=true}
   &DynamicMaterialParameters(
    Index 0 Param 1=&RandomRangeFloat(Minimum=0.2, Maximum=0.5)
    Module.Param0WriteEnabled=true
    Module.Param1WriteEnabled=true
    Module.Param2WriteEnabled=true
    Module.Param3WriteEnabled=true
   ) {enabled=false}
   &CameraOffset(Camera Offset Amount=15)
   &ShapeLocation(
    Cylinder Height=50
    Cylinder Height Midpoint=0.15
    Cylinder Radius=70
    MaskSpawnGroup=false
    Override Seed=false
    Shape Primitive=::NewEnumerator1
    Surface Only=false
    UseOwnerScale=true
   )
   &AddVelocity(
    Velocity=<0 0 1.1384>
    Velocity Speed Scale=&Multiply_Float(
     A=&RandomRangeFloat(Minimum=670.2246, Maximum=&MaskFloatBySpawnGroup(Masked Float=2029.0078, Passthrough Float=1000))
     B=$System.P_FXHeight
    )
   )
   /SetVariables(Particles.Variant1_DMP0:Vector4f=&MakeVector4(
     X=&RandomRangeFloat(Minimum=0, Maximum=1)
     Y=0
     Z=&RandomRangeFloat(Minimum=0, Maximum=1)
     W=0
    )) {isSetParameters=true}
   &AddVelocity(
    Cone Angle=160
    Inner Cone Angle=20
    Module.UseConeFalloff=false
    Module.UseVelDistribution=false
    Override Seed=false
    Velocity Mode=::NewEnumerator2
    Velocity Speed=&MaskFloatBySpawnGroup(Masked Float=0, Passthrough Float=800)
   ) {module=AddVelocity001}
   &InitialMeshOrientation(Use Rotation Vector=false)
  ]
  particleUpdateScript [
   &ParticleState()
   &ScaleMeshSize(Scale Factor=&VectorFromFloat(Value=&FloatFromCurve(FloatCurve=DataInterface() {propertyValues {Curve {
         defaultValue=0
         keys [
          {arriveTangent=9.0725, interpMode=RCIM_Cubic, leaveTangent=9.0725, tangentMode=RCTM_User, time=0, value=0}
          {arriveTangent=-2.5249, interpMode=RCIM_Cubic, leaveTangent=-2.5249, tangentMode=RCTM_User, time=0.1408, value=0.9847}
          {arriveTangent=-1.146, interpMode=RCIM_Cubic, leaveTangent=-1.146, time=0.6569, value=0.7799}
          {arriveTangent=-0.0405, interpMode=RCIM_Cubic, leaveTangent=-0.0405, tangentMode=RCTM_User, time=1, value=0}
         ]
        }}})))
   &UpdateMeshOrientation(
    Rotation Rate=&SelectFloatFromArray(
     Array Sampling Mode=::NewEnumerator0
     Direct Array Index=?
     Float Selection Array=DataInterface() {propertyValues {FloatData [-1,1]}}
     Random Seed=0
    )
    Rotation Vector=&RandomRangeVector(Minimum=<-1 -2 -1>, Maximum=<1 1 1>)
   )
   &GravityForce(Gravity=&MaskVectorBySpawnGroup(Masked Vector=<0 0 400>, Passthrough Vector=<0 0 -980>))
   &Drag(
    Drag=&MaskFloatBySpawnGroup(Masked Float=3, Passthrough Float=4)
    Use Linear Drag=true
    Use Rotational Drag=true
   )
   &SolveForcesAndVelocity(Clamp Velocity=false, Limit Acceleration=false)
   &FN_SubUVAnimation(End Frame=&Add_Integer(A=$Particles.RandomFrame, B=63), Start Frame=$Particles.RandomFrame)
   &ScaleSpriteSize(Scale Factor=&Vector2DFromFloat(Value=&FloatFromCurve(FloatCurve=DataInterface() {propertyValues {Curve {keys [
          {arriveTangent=1.5877, interpMode=RCIM_Cubic, leaveTangent=1.5877, tangentMode=RCTM_User, time=0, value=0.55}
          {arriveTangent=0.0897, interpMode=RCIM_Cubic, leaveTangent=0.0897, tangentMode=RCTM_User, time=1, value=1}
         ]}}})))
   &ScaleColor(
    Scale Alpha=&FloatFromCurve(FloatCurve=DataInterface() {propertyValues {Curve {keys [
         {interpMode=RCIM_Cubic, tangentMode=RCTM_User, time=0, value=0}
         {arriveTangent=-1.4898, interpMode=RCIM_Cubic, leaveTangent=-1.4898, tangentMode=RCTM_User, time=0.1, value=1.05}
         {arriveTangent=-4.9931, tangentMode=RCTM_User, time=1, value=0}
        ]}}})
    Scale RGB=&MakeVectorFromLinearColorRGB(Color=&LerpLinearColors(
      StartColor=&ColorFromCurve(ColorCurve=DataInterface() {propertyValues {
         AlphaCurve {keys [<at 0 1 RCIM_Cubic>]}
         BlueCurve {keys [<at 0 0 RCIM_Cubic>,<at 0.2905 0>,<at 0.4627 0.05 RCIM_Cubic>]}
         GreenCurve {keys [<at 0 1.6667 RCIM_Cubic>,<at 0.2905 0.4>,<at 0.4627 0.0542 RCIM_Cubic>]}
         RedCurve {keys [<at 0 50 RCIM_Cubic>,<at 0.2905 5>,<at 0.4627 0.1 RCIM_Cubic>]}
        }})
      EndColor=&ColorFromCurve(ColorCurve=DataInterface() {propertyValues {
         AlphaCurve {keys [<at 0 1 RCIM_Cubic>]}
         BlueCurve {keys [<at 0 0 RCIM_Cubic>,<at 0.2905 0>,<at 0.4627 0.004 RCIM_Cubic>]}
         GreenCurve {keys [<at 0 50 RCIM_Cubic>,<at 0.2905 5>,<at 0.4627 0.1 RCIM_Cubic>]}
         RedCurve {keys [<at 0 0.0075 RCIM_Cubic>,<at 0.2905 1.0434>,<at 0.4627 0.0386 RCIM_Cubic>]}
        }})
      LerpFactor=&MakeFloatFromInt(Int=$User.Color Variant)
     ))
    ScaleA=true
    ScaleRGB=true
   ) {module=ScaleColor001}
   &DynamicMaterialParameters(
    Index 0 Param 1=&FloatFromCurve(
     FloatCurve=DataInterface() {propertyValues {Curve {keys [
         {arriveTangent=2.8718, interpMode=RCIM_Cubic, leaveTangent=2.8718, tangentMode=RCTM_User, time=0, value=0}
         {arriveTangent=-0.3942, interpMode=RCIM_Cubic, leaveTangent=-0.3942, tangentMode=RCTM_User, time=1, value=1}
        ]}}}
     Scale Curve=0
    )
    Index 0 Param 2=250
    Module.Index0Param0Write=true
    Module.Index0Param1Write=true
    Module.Index0Param2Write=true
    Module.Index0Param3Write=true
   ) {module=DynamicMaterialParameters001}
  ]
  rendererClasses [&NiagaraSpriteRendererProperties, &NiagaraMeshRendererProperties]
  renderers [
   {rendererClass=&NiagaraSpriteRendererProperties}
   {rendererClass=&NiagaraMeshRendererProperties, rendererIndex=1}
  ]
 }
 StreaksBurst {
  emitterUpdateScript [
   &EmitterState()
   &SpawnBurst_Instantaneous(Spawn Count=15, Spawn Time=0.1, Use Spawn Probability=false, UseLoopCountLimit=false) {enabled=false}
   &SpawnPerFrame(
    Spawn:NiagaraBool=$System.ShouldSpawn
    Spawn Count:NiagaraInt32=&MakeIntFromFloat(Float=&Multiply_Float(A=15, B=$Engine.Emitter.SpawnCountScale), Conversion Mode=::NewEnumerator1)
    Use Spawn Probability=false
   )
  ]
  particleSpawnScript [
   &self:MANAGE_InitialManageVariables_01(Position:NiagaraPosition=$System.SpawnPosition, SpawnArrayIndex:NiagaraInt32=$System.SpawnArrayIndex) {module=MANAGE_InitialManageVariables_01}
   &PartitionParticles()
   &InitializeParticle(
    Color=$System.color_a
    Lifetime Max=0.4
    Lifetime Min=0.1
    Lifetime Mode=::NewEnumerator1
    Position=&TransformPosition(Position=$Particles.SpawnPosition, Source Space=::World, Destination Space=::Simulation)
    Position Mode=::NewEnumerator1
    Sprite Size Mode=::NewEnumerator3
    Sprite UV Mode=::NewEnumerator1
    Uniform Sprite Size=&Multiply_Float(A=57.2014, B=$System.P_FXRadius)
    UsePositionOffset=false
    Write Material Random=true
   )
   &CameraOffset(Camera Offset Amount=&SelectFloatFromArray(Direct Array Index=&SpawnGroup(), Float Selection Array=DataInterface() {propertyValues {FloatData [10,10]}}))
   &ShapeLocation(
    MaskSpawnGroup=false
    Offset Mode=::NewEnumerator0
    Override Seed=false
    Rotation Mode=::NewEnumerator1
    Sphere Radius=10
    UseOwnerScale=true
   )
   &AddVelocity(
    Module.Use Falloff Distance=false
    Origin Offset=&MakeVector(X=0, Y=0, Z=-7.5679)
    Override Seed=false
    Velocity Mode=::NewEnumerator1
    Velocity Origin=&TransformPosition(Position=$Particles.SpawnPosition, Source Space=::World, Destination Space=::Simulation)
    Velocity Speed=&Multiply_Float(A=&RandomRangeFloat(Minimum=0.264, Maximum=0.3088), B=1600)
   ) {module=AddVelocity002}
  ]
  particleUpdateScript [
   &self:MANAGE_ParticleDelay(DelayTime=0.1) {module=MANAGE_ParticleDelay}
   &ParticleState()
   &ApplyOwnerScaleToAttributes(Scale Initial Sprite Size=true)
   &ScaleSpriteSize(
    Scale Factor=&Multiply_Vector2D(A=&RandomRangeVector2D(Minimum=<2 2>, Maximum=<3 3>), B=<1 1>)
    Scale Sprite Size Mode=::NewEnumerator2
   )
   &ScaleSpriteSize(
    Non-Uniform Curve Scale=<1 4>
    Non-Uniform Curve Sprite Scale=DataInterface() {propertyValues {
      XCurve {keys [
        {arriveTangent=0.8609, interpMode=RCIM_Cubic, leaveTangent=0.8609, tangentMode=RCTM_User, time=0, value=0.6}
        {arriveTangent=0.5103, interpMode=RCIM_Cubic, leaveTangent=0.5103, tangentMode=RCTM_User, time=0.3, value=0.8}
        {arriveTangent=0.0224, interpMode=RCIM_Cubic, leaveTangent=0.0224, tangentMode=RCTM_User, time=1, value=1}
       ]}
      YCurve {keys [
        {arriveTangent=0.8801, interpMode=RCIM_Cubic, leaveTangent=0.8801, tangentMode=RCTM_User, time=0, value=0.3}
        {arriveTangent=0.6586, interpMode=RCIM_Cubic, leaveTangent=0.6586, tangentMode=RCTM_User, time=0.2, value=0.85}
        {arriveTangent=-0.4778, interpMode=RCIM_Cubic, leaveTangent=-0.4778, tangentMode=RCTM_User, time=1, value=0.75}
       ]}
     }}
    Scale Sprite Size Mode=::NewEnumerator3
   ) {module=ScaleSpriteSize001}
   &Drag(Drag=4, Use Linear Drag=true, Use Rotational Drag=true)
   &SolveForcesAndVelocity(Clamp Velocity=false, Limit Acceleration=false)
   &ScaleColor(
    Scale Alpha=&Multiply_Float(
     A=&FloatFromCurve(
      FloatCurve=DataInterface() {propertyValues {Curve {keys [
          {arriveTangent=9.0725, interpMode=RCIM_Cubic, leaveTangent=9.0725, tangentMode=RCTM_User, time=0, value=1}
          {arriveTangent=-2.5249, interpMode=RCIM_Cubic, leaveTangent=-2.5249, tangentMode=RCTM_User, time=0.1408, value=0.9847}
          {arriveTangent=-0.6199, interpMode=RCIM_Cubic, leaveTangent=-0.6199, time=0.3, value=0.95}
          {arriveTangent=-1.3571, interpMode=RCIM_Cubic, leaveTangent=-1.3571, time=0.6, value=0.7}
          {arriveTangent=-0.0405, interpMode=RCIM_Cubic, leaveTangent=-0.0405, tangentMode=RCTM_User, time=1, value=0}
         ]}}}
      Scale Curve=8
     )
     B=&Multiply_Float(A=1, B=1)
    )
    Scale RGB=&VectorFromFloat(Value=&Lerp_Float(
      A=1
      B=2.5
      Alpha=&FloatFromCurve(FloatCurve=DataInterface() {propertyValues {Curve {keys [
           {arriveTangent=106.6087, interpMode=RCIM_Cubic, leaveTangent=106.6087, tangentMode=RCTM_User, time=0, value=1}
           {arriveTangent=-2.5249, interpMode=RCIM_Cubic, leaveTangent=-2.5249, tangentMode=RCTM_User, time=0.1408, value=0.9847}
           {arriveTangent=-0.0405, interpMode=RCIM_Cubic, leaveTangent=-0.0405, tangentMode=RCTM_User, time=1, value=0}
          ]}}})
     ))
    ScaleA=true
    ScaleRGB=true
   ) {module=ScaleColor001}
   &ScaleColor(
    Scale Mode=::NewEnumerator1
    Scale RGBA=&MakeFloat4FromVectorAndFloat("Vector (RGB)"=<1 1 1>, "Float (Alpha)"=1)
    ScaleRGBA=true
   ) {module=ScaleColor002}
   &DynamicMaterialParameters(
    Index 0 Param 1=&FloatFromCurve(FloatCurve=DataInterface() {propertyValues {Curve {keys [
         {arriveTangent=-0.1846, time=0, value=0}
         {arriveTangent=1.3385, interpMode=RCIM_Cubic, leaveTangent=0.1195, tangentMode=RCTM_Break, time=0.3, value=0}
         {arriveTangent=1.4286, interpMode=RCIM_Cubic, leaveTangent=1.4286, time=0.8, value=0.5}
         {arriveTangent=3.0091, interpMode=RCIM_Cubic, leaveTangent=3.0091, tangentMode=RCTM_User, time=1, value=1}
        ]}}})
    Index 0 Param 4=&FloatFromCurve(FloatCurve=DataInterface() {propertyValues {Curve {keys [<at 0 0 RCIM_Cubic>,<at 1 1 RCIM_Cubic>]}}}, Scale Curve=0.8)
    Module.Index0Param0Write=true
    Module.Index0Param1Write=true
    Module.Index0Param2Write=true
    Module.Index0Param3Write=true
   )
  ]
  rendererClasses [&NiagaraSpriteRendererProperties]
  renderers [{rendererClass=&NiagaraSpriteRendererProperties}]
 }
 Streaks_Fat {
  emitterUpdateScript [
   &EmitterState()
   &SpawnBurst_Instantaneous(Spawn Count=13, Use Spawn Probability=false) {enabled=false}
   &SpawnPerFrame(
    Spawn:NiagaraBool=$System.ShouldSpawn
    Spawn Count:NiagaraInt32=&MakeIntFromFloat(Float=&Multiply_Float(A=13, B=$Engine.Emitter.SpawnCountScale), Conversion Mode=::NewEnumerator1)
    Use Spawn Probability=false
   )
  ]
  particleSpawnScript [
   &self:MANAGE_InitialManageVariables_02(Position:NiagaraPosition=$System.SpawnPosition, SpawnArrayIndex:NiagaraInt32=$System.SpawnArrayIndex) {module=MANAGE_InitialManageVariables_02}
   &InitializeParticle(
    Color=$System.color_c
    Lifetime Max=0.3
    Lifetime Min=0.2
    Lifetime Mode=::NewEnumerator1
    Position=&TransformPosition(Position=$Particles.SpawnPosition, Source Space=::World, Destination Space=::Simulation)
    Position Mode=::NewEnumerator1
    Sprite Size=&Multiply_Vector2D_ByFloat(
     Float=$System.P_FXRadius
     Vector2D=&UniformRangedVector2D(Minimum=<80 93.7853>, Maximum=<99.3215 207.1198>)
    )
    Sprite Size Mode=::NewEnumerator4
    UsePositionOffset=false
    Write Material Random=true
   ) {module=InitializeParticle001}
   &ShapeLocation(
    Cylinder Height=1
    Cylinder Height Midpoint=0.1
    Cylinder Radius=30
    MaskSpawnGroup=false
    Offset=<0 0 -20>
    Offset Mode=::NewEnumerator0
    Override Seed=false
    Shape Primitive=::NewEnumerator1
    Surface Only=false
    UseOwnerScale=true
   )
   &AddVelocity(
    Module.Use Falloff Distance=false
    Origin Offset=&MakeVector(X=0, Y=0, Z=-42.0136)
    Override Seed=false
    Velocity Mode=::NewEnumerator1
    Velocity Origin=&TransformPosition(Position=$Particles.SpawnPosition, Source Space=::World, Destination Space=::Simulation)
    Velocity Speed=&Multiply_Float(A=&RandomRangeFloat(Minimum=0.2, Maximum=1), B=1159.5756)
   ) {module=AddVelocity002}
   &CameraOffset(Camera Offset Amount=100)
  ]
  particleUpdateScript [
   &ParticleState()
   &ApplyOwnerScaleToAttributes(Scale Initial Sprite Size=true)
   &ScaleSpriteSize(Scale Sprite Size Mode=::NewEnumerator0, Uniform Scale Factor=1.2) {module=ScaleSpriteSize001}
   &SolveForcesAndVelocity(Clamp Velocity=false, Limit Acceleration=false)
   &ScaleSpriteSize(Scale Factor=&Vector2DFromCurve(
     Vector2Curve=DataInterface() {propertyValues {
       XCurve {keys [<at 0 0.75 RCIM_Cubic>, {arriveTangent=-0.75, interpMode=RCIM_Cubic, leaveTangent=-0.75, time=0.2, value=1.625}, <at 1 0 RCIM_Cubic>]}
       YCurve {keys [<at 0 0 RCIM_Cubic>,<at 1 2 RCIM_Cubic>]}
      }}
     ScaleCurve=&Multiply_Vector2D_ByFloat(
      Float=&Multiply_Float(
       A=&FN_CameraDistance(
        0Dist:NiagaraFloat=400
        1Dist:NiagaraFloat=800
        Camera Query:NiagaraDataInterfaceCamera=DataInterface()
        Position:NiagaraPosition=$Particles.Position
       )
       B=1
      )
      Vector2D=<1 1>
     )
    ))
   &ScaleColor(
    Scale Alpha=&FloatFromCurve(
     FloatCurve=DataInterface() {propertyValues {Curve {keys [<at 0 1 RCIM_Cubic>,<at 1 0>]}}}
     Scale Curve=&Multiply_Float(
      A=&FN_CameraDistance(
       0Dist:NiagaraFloat=400
       1Dist:NiagaraFloat=800
       Camera Query:NiagaraDataInterfaceCamera=DataInterface()
       Position:NiagaraPosition=$Particles.Position
      )
      B=1
     )
    )
    Scale RGB=&VectorFromFloat(Value=&FloatFromCurve(
      FloatCurve=DataInterface() {propertyValues {Curve {keys [<at 0 10 RCIM_Cubic>,<at 1 1 RCIM_Cubic>]}}}
      Scale Curve=&Multiply_Float(A=1, B=1)
     ))
    ScaleA=true
    ScaleRGB=true
   )
  ]
  rendererClasses [&NiagaraSpriteRendererProperties]
  renderers [{rendererClass=&NiagaraSpriteRendererProperties}]
 }
}
summary {
 systemName=NS_FirePetal_WanderingBoss_Bombard_AOE_Burst_MergedV
 emitters [
  {emitterName=DecalShockWave, rendererClasses [&NiagaraMeshRendererProperties, &NiagaraSpriteRendererProperties]}
  {
   emitterName=LightAndDecal001
   rendererClasses [
    &NiagaraDecalRendererProperties
    &NiagaraMeshRendererProperties
    &NiagaraSpriteRendererProperties
    &NiagaraLightRendererProperties
   ]
  }
  {emitterName=LightAndDecal, rendererClasses [&NiagaraSpriteRendererProperties, &NiagaraLightRendererProperties]}
  {emitterName=Smoke_Wisp001, rendererClasses [&NiagaraSpriteRendererProperties, &NiagaraMeshRendererProperties]}
  {emitterName=StreaksBurst, rendererClasses [&NiagaraSpriteRendererProperties]}
  {emitterName=Streaks_Fat, rendererClasses [&NiagaraSpriteRendererProperties]}
  {emitterName=Smoke_Ring001, rendererClasses [&NiagaraSpriteRendererProperties]}
  {emitterName=Particle-Bright, simTarget=GPUComputeSim, rendererClasses [&NiagaraSpriteRendererProperties]}
  {emitterName=Particle-Dissolve, simTarget=GPUComputeSim, rendererClasses [&NiagaraSpriteRendererProperties]}
 ]
 userVariables [
  User.bIsFrontEnd {type=NiagaraBool}
  User.bIsFrontEndPreview {type=NiagaraBool}
  User.Color Variant {type=NiagaraInt32}
  User.Delay {type=NiagaraFloat, defaultValue=0, description=""}
  User.ElementSizeScale {type=NiagaraFloat, defaultValue=1, description=""}
  User.FireColor {type=LinearColor, description="", defaultValue {struct=LinearColor, value=<rgba 5 0.3889 0 1>}}
  User.FXHeight {type=NiagaraFloat, defaultValue=7, description=""}
  User.FXRadius {type=NiagaraFloat, defaultValue=3, description=""}
  User.IsAttach {type=NiagaraBool}
  User.IsUltra {type=NiagaraBool, defaultValue=true}
  User.SpawnArray {
   type=NiagaraDataInterfaceArrayPosition
   description=""
   defaultValue {
    struct=NiagaraExt_VariableValue_DataInterface
    value {
     dataInterface=&self:NiagaraDataInterfaceArrayPosition_0
     dataInterfaceClass=&NiagaraDataInterfaceArrayPosition
    }
   }
  }
  User.SpawnCounter {
   type=NiagaraDataInterfaceSimpleCounter
   description=""
   defaultValue {
    struct=NiagaraExt_VariableValue_DataInterface
    value {
     dataInterface=&self:NiagaraDataInterfaceSimpleCounter_0
     dataInterfaceClass=/Script/Niagara.NiagaraDataInterfaceSimpleCounter
    }
   }
  }
  User.SpawnDecal {type=NiagaraBool, defaultValue=true}
  User.WaterFX {type=NiagaraBool}
 ]
}
