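# Niagara telegraph for the FirePetal wandering boss's bombard AOE, managing many strike points in one system via user arrays.
# System MANAGE modules trigger per-point decal, mesh, glow, light and delayed falling-glow spawns, while smoke-ring spawn rate scales with list length.
# Decals and glows intensify on a shared 2.8-second normalized age and die via ExplodeArray index lookups.
# Color Variant shifts glow palette; platform checks trim extra decal layers.
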NiagaraSystem /FirePetal_WanderingBoss/VFX/NS_FirePetal_WanderingBoss_Bombard_AOE_Idle_ManageVersion
dependencies {
 usedDataInterfaces [
  &NiagaraDataInterfacePlatformSet
  &NiagaraDataInterfaceCurve
  &NiagaraDataInterfaceColorCurve
  &NiagaraDataInterfaceArrayColor
 ]
 usedDynamicInputs [&MaskFloatBySpawnGroup, &SelectLinearColorFromArray]
 usedModules [
  &self:SystemSpawnScript.SystemScriptSource.SystemScriptGraph.NiagaraNodeAssignment_1.SetVariables_#1
  &SystemState
  &self:SystemSpawnScript.SystemScriptSource.SystemScriptGraph.NiagaraNodeAssignment_0.SetVariables_#2
  &self:MANAGE_ShouldSpawnThisFrame
  &self:MANAGE_UpdateCounter
 ]
}
emitters {
 Decal {
  emitterUpdateScript [
   &EmitterState()
   &SpawnPerFrame(Spawn:NiagaraBool=$System.ShouldSpawn, Use Spawn Probability=false)
   &SpawnPerFrame(
    Spawn:NiagaraBool=$System.ShouldSpawn
    Spawn Count:NiagaraInt32=&MakeCustomIntFromBool(
     Bool=&IsPlatformSetActive(New Platform Set=DataInterface() {propertyValues {Platforms {
         qualityLevelMask=-32
         deviceProfileStates [
          {profileName=Mobile, qualityLevelMask=4294967295, setQualityLevelMask=4294967295}
          {profileName=Switch, qualityLevelMask=4294967295, setQualityLevelMask=4294967295}
         ]
        }}})
     True INT=0
     False INT=1
    )
    Spawn Group=1
    Use Spawn Probability=false
   ) {module=SpawnPerFrame002}
   &SpawnPerFrame(
    Spawn:NiagaraBool=$System.ShouldSpawn
    Spawn Count:NiagaraInt32=&MakeCustomIntFromBool(
     Bool=&IsPlatformSetActive(New Platform Set=DataInterface() {propertyValues {Platforms {
         qualityLevelMask=-32
         deviceProfileStates [
          {profileName=Mobile, qualityLevelMask=4294967295, setQualityLevelMask=4294967295}
          {profileName=Switch, qualityLevelMask=4294967295, setQualityLevelMask=4294967295}
         ]
        }}})
     True INT=0
     False INT=1
    )
    Spawn Group=2
    Use Spawn Probability=false
   ) {module=SpawnPerFrame001}
  ]
  particleSpawnScript [
   &self:MANAGE_InitialManageVariables(Position:NiagaraPosition=$System.SpawnPosition, SpawnArrayIndex:NiagaraInt32=$System.SpawnArrayIndex) {module=MANAGE_InitialManageVariables}
   &InitializeParticle(
    Color=$System.color_d
    Mesh Scale Mode=::NewEnumerator3
    Mesh Uniform Scale=&Multiply_Float(A=$User.Radius, B=4)
    Position=&TransformPosition(Position=$System.SpawnPosition, Source Space=::World, Destination Space=::Simulation)
    Position Mode=::NewEnumerator1
    Position Offset=&MaskVectorBySpawnGroup(Masked Vector=<0 0 -10>, Passthrough Vector=<0 0 0>, Spawn Group Mask=2)
    UsePositionOffset=true
    Write Material Random=true
   )
   &CameraOffset(Camera Offset Amount=20)
   /SetVariables(
    Particles.DecalFade=1
    Particles.DecalSize:Vector3f=&MaskVectorBySpawnGroup(Masked Vector=<300 250 250>, Passthrough Vector=<300 380 380>)
   ) {isSetParameters=true}
   /SetVariables(Particles.VisibilityTag=&SpawnGroup()) {isSetParameters=true}
  ]
  particleUpdateScript [
   &ParticleState(Kill Particles When Lifetime Has Elapsed=false, Let Infinitely Lived Particles Die When Emitter Deactivates=true)
   /SetVariables(Particles.TelAge:NiagaraFloat=&Divide_Float(A=$Particles.Age, B=2.8)) {isSetParameters=true}
   &ScaleMeshSize(Scale Factor=<1 1 0.3>)
   &DynamicMaterialParameters(
    Index 0 Param 1=&Multiply_Float(
     A=&Waveform(
      Global Phase=$Particles.Age
      "Waveform [1]"=::NewEnumerator12
      "[1] Phase"=$Particles.Age
      "[1] Frequency"=10
      "[1] Amplitude Min/Max"=<0.5 1>
     )
     B=&FloatFromCurve(
      FloatCurve=DataInterface() {propertyValues {Curve {keys [
          <at 0 0 RCIM_Cubic>
          {interpMode=RCIM_Cubic, leaveTangent=58.4456, tangentMode=RCTM_Break, time=0.7, value=0.5}
          {arriveTangent=0.0425, leaveTangent=5.5797, tangentMode=RCTM_User, time=1, value=5}
         ]}}}
      CurveIndex=$Particles.TelAge
     )
    )
    Index 0 Param 2=&MakeFloatFromVector(Vector=Linked() {linkedVariable {name=Particles.SpawnPosition, type=NiagaraPosition}}, Channel=::NewEnumerator0)
    Index 1 Param 1=&Multiply_Float(A=$User.Radius, B=3)
    Index 1 Param 2=&Multiply_Float(A=$User.Radius, B=3)
    Module.Index0Param0Write=true
    Module.Index0Param1Write=true
    Module.Index0Param2Write=true
    Module.Index0Param3Write=true
    Module.Index1Param0Write=true
    Module.Index1Param1Write=true
    Module.Index1Param2Write=true
    Module.Index1Param3Write=true
    Write Parameter Index 1=true
   ) {module=DynamicMaterialParameters001}
   &SpriteFacingAndAlignment(Sprite Facing=<0 0 1>, WriteAlignment=true, WriteFacing=true) {module=SpriteFacingAndAlignment001}
   /SetVariables(Particles.Color=&MaskLinearColorBySpawnGroup(
     Masked Linear Color=&MultiplyLinearColor(
      Linear Color=$System.color_a
      Scale Factor=&MakeVector4(
       X=1
       Y=1
       Z=1
       W=&FloatFromCurve(FloatCurve=DataInterface() {propertyValues {Curve {keys [<at 0 0>,<at 0.4 1>]}}}, CurveIndex=$Particles.TelAge)
      )
     )
     Passthrough Linear Color=&MaskLinearColorBySpawnGroup(
      Masked Linear Color=&MakeLinearColor(
       R:NiagaraFloat=&FloatFromCurve(
        FloatCurve=DataInterface() {propertyValues {Curve {keys [{interpMode=RCIM_Cubic, leaveTangent=3.5502, tangentMode=RCTM_Break, time=0, value=-2}, <at 0.5 1>]}}}
        CurveIndex=$Particles.TelAge
        Scale Curve=0.26
       )
       G:NiagaraFloat=1
       B:NiagaraFloat=&FloatFromCurve(
        FloatCurve=DataInterface() {propertyValues {Curve {keys [<at 0 0>,<at 0.05 1>]}}}
        CurveIndex=&Divide_Float(A=$Particles.Age, B=9)
        Scale Curve=3
       )
       A:NiagaraFloat=&FloatFromCurve(
        FloatCurve=DataInterface() {propertyValues {Curve {keys [{interpMode=RCIM_Cubic, leaveTangent=2.9459, tangentMode=RCTM_Break, time=0, value=0}, <at 0.5 1>]}}}
        CurveIndex=$Particles.TelAge
       )
      )
      Passthrough Linear Color=&MakeLinearColor(
       R:NiagaraFloat=&FloatFromCurve(
        FloatCurve=DataInterface() {propertyValues {Curve {keys [{interpMode=RCIM_Cubic, leaveTangent=3.5502, tangentMode=RCTM_Break, time=0, value=-2}, <at 0.5 1>]}}}
        CurveIndex=$Particles.TelAge
        Scale Curve=0.26
       )
       G:NiagaraFloat=0
       B:NiagaraFloat=1
       A:NiagaraFloat=1
      )
     )
     Spawn Group Mask=2
    )) {isSetParameters=true}
   &self:MANAGE_ShouldKill(Bool Array:NiagaraDataInterfaceArrayBool=$User.ExplodeArray, MyIndex:NiagaraInt32=$Particles.ArrayIndex) {module=MANAGE_ShouldKill}
   &KillParticles(Kill Particles=$Particles.ShouldKill)
  ]
  rendererClasses [&NiagaraDecalRendererProperties, &NiagaraMeshRendererProperties]
  renderers [
   {rendererClass=&NiagaraDecalRendererProperties, rendererIndex=0}
   {rendererClass=&NiagaraDecalRendererProperties, rendererIndex=1}
   {rendererClass=&NiagaraMeshRendererProperties, rendererIndex=2}
  ]
 }
 FalldownGlow {
  emitterUpdateScript [
   &EmitterState()
   &SpawnPerFrame(Spawn:NiagaraBool=$System.ShouldSpawn, Use Spawn Probability=false)
  ]
  particleSpawnScript [
   &self:MANAGE_InitialManageVariables(Position:NiagaraPosition=$System.SpawnPosition, SpawnArrayIndex:NiagaraInt32=$System.SpawnArrayIndex) {module=MANAGE_InitialManageVariables}
   &InitializeParticle(
    Color=$System.color_c
    Position=&TransformPosition(Position=$System.SpawnPosition, Source Space=::World, Destination Space=::Simulation)
    Position Mode=::NewEnumerator1
    Sprite Size Mode=::NewEnumerator3
    Uniform Sprite Size=430.8955
    UsePositionOffset=false
    Write Material Random=true
   )
  ]
  particleUpdateScript [
   &self:MANAGE_ParticleDelay(DelayTime=1.4) {module=MANAGE_ParticleDelay}
   &ParticleState()
   &ScaleSpriteSize(Uniform Curve Sprite Scale=DataInterface() {propertyValues {Curve {keys [
        {interpMode=RCIM_Cubic, tangentMode=RCTM_User, time=0, value=0}
        {arriveTangent=-0.4593, leaveTangent=-2, tangentMode=RCTM_User, time=0.9, value=1}
       ]}}}) {module=ScaleSpriteSize001}
   &ScaleSpriteSize(
    Scale Sprite Size Mode=::NewEnumerator0
    Uniform Scale Factor=&Waveform(
     Global Phase=$Particles.Age
     "Waveform [1]"=::NewEnumerator12
     "[1] Phase"=$Particles.Age
     "[1] Frequency"=50
     "[1] Amplitude Min/Max"=<0.2 1>
    )
   )
   /SetVariables(Particles.Position=&AddVectorToPosition(
     Position=&TransformPosition(Position=$Particles.SpawnPosition, Source Space=::World, Destination Space=::Simulation)
     Vector=&MakeVector(
      X=0
      Y=0
      Z=&FloatFromCurve(
       FloatCurve=DataInterface() {propertyValues {Curve {keys [
           {interpMode=RCIM_Cubic, leaveTangent=-0.1917, tangentMode=RCTM_Break, time=0, value=1}
           {arriveTangent=-0.3373, tangentMode=RCTM_User, time=1, value=0}
          ]}}}
       Scale Curve=200
      )
     )
    )) {isSetParameters=true}
   &ScaleColor(
    Scale Alpha=&FloatFromCurve(FloatCurve=DataInterface() {propertyValues {Curve {keys [
         {interpMode=RCIM_Cubic, tangentMode=RCTM_User, time=0, value=0}
         {arriveTangent=3.4267, interpMode=RCIM_Cubic, leaveTangent=3.4267, tangentMode=RCTM_User, time=0.8, value=1}
        ]}}})
    ScaleA=true
    ScaleRGB=true
   )
  ]
  rendererClasses [&NiagaraSpriteRendererProperties]
  renderers [{rendererClass=&NiagaraSpriteRendererProperties}]
 }
 GlowAndEnergy {
  emitterUpdateScript [
   &EmitterState()
   &SpawnPerFrame(Spawn:NiagaraBool=$System.ShouldSpawn, Use Spawn Probability=false)
   &SpawnPerFrame(Spawn:NiagaraBool=$System.ShouldSpawn, Spawn Group=1, Use Spawn Probability=false) {module=SpawnPerFrame002}
   &SpawnPerFrame(Spawn:NiagaraBool=$System.ShouldSpawn, Spawn Group=2, Use Spawn Probability=false) {module=SpawnPerFrame003}
   &SpawnPerFrame(Spawn:NiagaraBool=$System.ShouldSpawn, Spawn Group=3, Use Spawn Probability=false) {module=SpawnPerFrame001}
  ]
  particleSpawnScript [
   &self:MANAGE_InitialManageVariables(Position:NiagaraPosition=$System.SpawnPosition, SpawnArrayIndex:NiagaraInt32=$System.SpawnArrayIndex) {module=MANAGE_InitialManageVariables}
   &InitializeParticle(
    Color=&LerpLinearColors(
     StartColor=&MergeHelper_ColorSwitch(
      0=<rgba 1 0.0833 0 0.4>
      1=<rgba 0.05 0.0271 0.025 1>
      2=<rgba 0.6 0.06 0 6>
      3=<rgba 10 1.9167 0.3 1>
      4=<rgba 0 0 0 0>
      Index=&SpawnGroup()
      Num=::NewEnumerator3
     )
     EndColor=&MergeHelper_ColorSwitch(
      0=<rgba 0.0124 1 0 0.4>
      1=<rgba 0.023 0.05 0.0246 1>
      2=<rgba 0.1551 0.6 0 6>
      3=<rgba 0 10 1.2144 1>
      4=<rgba 0 0 0 0>
      Index=&SpawnGroup()
      Num=::NewEnumerator3
     )
     LerpFactor=&MakeFloatFromInt(Int=$User.Color Variant)
    )
    Lifetime=5
    Mesh Scale=<0.5 0.5 0.7>
    Mesh Scale Mode=::NewEnumerator4
    Position=&TransformPosition(Position=$System.SpawnPosition, Source Space=::World, Destination Space=::Simulation)
    Position Mode=::NewEnumerator1
    Position Offset=&MergeHelper_VectorSwitch(0=<0 0 0>, 1=<0 0 -17>, 2=<0 0 0>, 3=<0 0 -8>, 4=<0 0 0>, Index=&SpawnGroup(), Num=::NewEnumerator3)
    Sprite Size=&MergeHelper_Vector2Switch(
     0=&Vector2DFromFloat(Value=427.8739)
     1=&Vector2DFromFloat(Value=427.8739)
     2=<250 350>
     3=&Vector2DFromFloat(Value=450)
     4=<0 0>
     Index=&SpawnGroup()
     Num=::NewEnumerator3
    )
    Sprite Size Mode=::NewEnumerator4
    UsePositionOffset=true
    Write Material Random=true
   )
   /SetVariables(Particles.VisibilityTag=&SpawnGroup()) {isSetParameters=true}
  ]
  particleUpdateScript [
   &ParticleState(Kill Particles When Lifetime Has Elapsed=false, Let Infinitely Lived Particles Die When Emitter Deactivates=true)
   /SetVariables(Particles.TelAge:NiagaraFloat=&Divide_Float(A=$Particles.Age, B=2.8)) {isSetParameters=true}
   &CameraOffset(Camera Offset Amount=&MergeHelper_FloatSwitch(0=50, 1=50, 2=50, 3=100, 4=0, Index=&SpawnGroup(), Num=::NewEnumerator3))
   &FN_Light_Attributes(
    Color=&Multiply_LinearColorByFloat(Linear Color=$System.color_decal1, Float=1.5)
    Create Color Attribute=true
    Create Exponent Attribute=true
    Create Position Attribute=true
    Create Radius Attribute=true
    Position=&AddVectorToPosition(
     Position=&TransformPosition(Position=$Particles.SpawnPosition, Source Space=::World, Destination Space=::Local)
     Vector=<0 0 100>
    )
    Radius=500
    Scale Color=&FloatFromCurve(
     FloatCurve=DataInterface() {propertyValues {Curve {keys [{interpMode=RCIM_Cubic, tangentMode=RCTM_Break, time=0, value=0}, <at 0.7 1>]}}}
     CurveIndex=$Particles.TelAge
     Scale Curve=300000
    )
   )
   &ScaleColor(
    Scale Alpha=&MergeHelper_FloatSwitch(
     0=&FloatFromCurve(
      FloatCurve=DataInterface() {propertyValues {Curve {keys [
          {interpMode=RCIM_Cubic, leaveTangent=0.9492, tangentMode=RCTM_Break, time=0, value=0}
          {arriveTangent=0.5703, interpMode=RCIM_Cubic, tangentMode=RCTM_Break, time=0.8, value=1}
         ]}}}
      CurveIndex=$Particles.TelAge
     )
     1=&FloatFromCurve(
      FloatCurve=DataInterface() {propertyValues {Curve {keys [
          {interpMode=RCIM_Cubic, leaveTangent=0.9957, tangentMode=RCTM_Break, time=0, value=0}
          {arriveTangent=2.1829, interpMode=RCIM_Cubic, tangentMode=RCTM_Break, time=0.8, value=1}
         ]}}}
      CurveIndex=$Particles.TelAge
     )
     2=&FloatFromCurve(
      FloatCurve=DataInterface() {propertyValues {Curve {keys [
          {interpMode=RCIM_Cubic, leaveTangent=2.9668, tangentMode=RCTM_Break, time=0, value=0}
          {arriveTangent=0.5703, interpMode=RCIM_Cubic, tangentMode=RCTM_Break, time=0.8, value=1}
         ]}}}
      CurveIndex=$Particles.TelAge
     )
     3=&FloatFromCurve(
      FloatCurve=DataInterface() {propertyValues {Curve {keys [{interpMode=RCIM_Cubic, leaveTangent=9.5547, tangentMode=RCTM_Break, time=0.6, value=0}, {tangentMode=RCTM_User, time=1, value=1}]}}}
      CurveIndex=$Particles.TelAge
      Scale Curve=&RandomRangeFloat(Minimum=0.2, Maximum=2, Evaluation Type=::NewEnumerator1)
     )
     4=0
     Index=&SpawnGroup()
     Num=::NewEnumerator3
    )
    ScaleA=true
    ScaleRGB=true
   )
   &ScaleMeshSize(Scale Factor=&VectorFromFloat(Value=&FloatFromCurve(
      FloatCurve=DataInterface() {propertyValues {Curve {keys [{interpMode=RCIM_Cubic, leaveTangent=1.3787, tangentMode=RCTM_Break, time=0, value=0.5}, {tangentMode=RCTM_User, time=1, value=1}]}}}
      CurveIndex=$Particles.TelAge
     )))
   &Timeline(
    Initial Start Time:NiagaraFloat=&RandomRangeFloat(Minimum=0, Maximum=200)
    Play Rate=&FloatFromCurve(
     FloatCurve=DataInterface() {propertyValues {Curve {keys [{interpMode=RCIM_Cubic, tangentMode=RCTM_Break, time=0, value=1}, {arriveTangent=4.8541, tangentMode=RCTM_User, time=0.7, value=4}]}}}
     CurveIndex=$Particles.TelAge
    )
    Timeline End=100000
   )
   &ScaleSpriteSize(
    Scale Factor=&MergeHelper_Vector2Switch(
     0=<1 1>
     1=<1 1>
     2=&MakeVector2D(
      X=1
      Y=&FloatFromCurve(
       FloatCurve=DataInterface() {propertyValues {Curve {keys [{interpMode=RCIM_Cubic, leaveTangent=2.6094, tangentMode=RCTM_Break, time=0, value=0}, {tangentMode=RCTM_User, time=1, value=1}]}}}
       CurveIndex=$Particles.TelAge
      )
     )
     3=<1 1>
     4=<0 0>
     Index=&SpawnGroup()
     Num=::NewEnumerator3
    )
    Scale Sprite Size Mode=::NewEnumerator2
   )
   &DynamicMaterialParameters(
    Index 0 Param 1=&MergeHelper_FloatSwitch(0=1, 1=1, 2=$Particles.Timeline.Time, 3=0, 4=0, Index=&SpawnGroup(), Num=::NewEnumerator3)
    Index 0 Param 2=1
    Index 0 Param 3=1
    Index 0 Param 4=1
    Module.Index0Param0Write=true
    Module.Index0Param1Write=true
    Module.Index0Param2Write=true
    Module.Index0Param3Write=true
   )
   &SpriteFacingAndAlignment(Sprite Alignment=<0 0 1>, WriteAlignment=true, WriteFacing=false)
   &self:MANAGE_ShouldKill(Bool Array:NiagaraDataInterfaceArrayBool=$User.ExplodeArray, MyIndex:NiagaraInt32=$Particles.ArrayIndex) {module=MANAGE_ShouldKill}
   &KillParticles(Kill Particles=$Particles.ShouldKill)
  ]
  rendererClasses [
   &NiagaraSpriteRendererProperties
   &NiagaraLightRendererProperties
   &NiagaraMeshRendererProperties
  ]
  renderers [
   {rendererClass=&NiagaraSpriteRendererProperties}
   {rendererClass=&NiagaraLightRendererProperties, rendererIndex=1}
   {rendererClass=&NiagaraMeshRendererProperties, rendererIndex=2}
   {rendererClass=&NiagaraSpriteRendererProperties, rendererIndex=3}
   {rendererClass=&NiagaraSpriteRendererProperties, rendererIndex=4}
  ]
 }
 Smoke_Ring001 {
  emitterSpawnScript [/SetVariables(Emitter.LightingCenter:Vector4f=&MakeFloat4FromVectorAndFloat(
     "Vector (RGB)"=&TransformPosition(Position=<0 0 0>, Source Space=::Local, Destination Space=::World)
     "Float (Alpha)"=1
    )) {isSetParameters=true}]
  emitterUpdateScript [
   &EmitterState()
   &self:MANAGE_SpawnRateFromList(SpawnRateList:NiagaraDataInterfaceArrayPosition=$User.SpawnRateList) {module=MANAGE_SpawnRateFromList}
   &SpawnRate(
    SpawnRate=&Multiply_Float(A=6, B=&MakeFloatFromInt(Int=$Emitter.ListLength))
    Use Spawn Probability=false
   )
  ]
  particleSpawnScript [
   &self:MANAGE_InitialRateManageVariables(
    Position Array:NiagaraDataInterfaceArrayPosition=$User.SpawnRateList
    StartSpawnTimeArray:NiagaraDataInterfaceArrayFloat=$User.SpawnRateStartTime
   ) {module=MANAGE_InitialRateManageVariables}
   &InitializeParticle(
    Color=$System.color_d
    Lifetime=&RandomRangeFloat(Minimum=1.5, Maximum=2)
    Position=&TransformPosition(Position=$Particles.ManagePos, Source Space=::World, Destination Space=::Simulation)
    Position Mode=::NewEnumerator1
    Position Offset=<0 0 -10>
    Sprite Rotation Mode=::NewEnumerator2
    Sprite Size Mode=::NewEnumerator3
    Uniform Sprite Size=&Multiply_Float(
     A=&RandomRangeFloat(Minimum=70, Maximum=80)
     B=&Multiply_Float(A=1, B=1)
    )
    UsePositionOffset=true
    Write Material Random=true
   ) {module=InitializeParticle001}
   &ShapeLocation(
    Handle Radius=1
    Large Radius=40
    MaskSpawnGroup=false
    Non Uniform Scale=<1 1 0>
    Shape Primitive=::NewEnumerator3
    Torus Distribution Mode:ENiagaraTorusDistributionMode=::NewEnumerator1
    U Position=&Add_Float(A=$Emitter.Age, B=&Multiply_Float(A=$Particles.SpawnStartTime, B=125.287))
    UseOwnerScale=true
   )
   &AddVelocity(
    Module.Use Falloff Distance=false
    Origin Offset=<0 0 -6>
    Override Seed=false
    Velocity Mode=::NewEnumerator1
    Velocity Origin=&TransformPosition(Position=$Particles.ManagePos, Source Space=::World, Destination Space=::Simulation)
    Velocity Speed=90
   )
  ]
  particleUpdateScript [
   /SetVariables(Particles.TelAge:NiagaraFloat=&Divide_Float(A=&Subtract_Float(A=$NPC.FortniteNPC.RealGameTime, B=$Particles.SpawnStartTime), B=2.8)) {isSetParameters=true}
   &ParticleState()
   &ScaleSpriteSize(Scale Factor=<1.2 1>, Scale Sprite Size Mode=::NewEnumerator2) {module=ScaleSpriteSize001}
   &CameraOffset(Camera Offset Amount=&FloatFromCurve(
     FloatCurve=DataInterface() {propertyValues {Curve {keys [<at 0 0>,<at 1 1>]}}}
     CurveIndex=&FN_LOD_Distance_Remap(Start Distance=100, End Distance=500, Minimum Output=0, Maximum Output=1, Clamp Output=true)
     Scale Curve=100
    ))
   &Drag(Drag=0.5, Use Linear Drag=true, Use Rotational Drag=false)
   &GravityForce(Gravity=<0 0 10>)
   &SolveForcesAndVelocity(Clamp Velocity=false, Limit Acceleration=false)
   &ScaleSpriteSize(Scale Factor=&Vector2DFromFloat(Value=&FloatFromCurve(
      FloatCurve=DataInterface() {propertyValues {Curve {keys [
          {interpMode=RCIM_Cubic, leaveTangent=1.7304, tangentMode=RCTM_Break, time=0, value=0.5}
          {arriveTangent=-0.0555, interpMode=RCIM_Cubic, leaveTangent=-0.0555, tangentMode=RCTM_User, time=0.3, value=1}
          <at 0.8 1>
         ]}}}
      Scale Curve=2
     )))
   /SetVariables(Particles.Color_ParticleEnergy:LinearColor=&ColorFromCurve(ColorCurve=DataInterface() {propertyValues {
       AlphaCurve {keys [<at 0 1 RCIM_Cubic>]}
       BlueCurve {keys [<at 0.3364 2>,<at 0.749 0.4167 RCIM_Cubic>]}
       GreenCurve {keys [<at 0.3364 0.35>,<at 0.749 0 RCIM_Cubic>]}
       RedCurve {keys [<at 0.3364 0.2>,<at 0.749 5 RCIM_Cubic>]}
      }})) {isSetParameters=true}
   &DynamicMaterialParameters(
    Index 0 Param 1=&FloatFromCurve(FloatCurve=DataInterface() {propertyValues {Curve {keys [{tangentMode=RCTM_User, time=0, value=0}, {tangentMode=RCTM_User, time=1, value=1}]}}})
    Index 0 Param 3=&FloatFromCurve(
     FloatCurve=DataInterface() {propertyValues {Curve {keys [
         {interpMode=RCIM_Cubic, tangentMode=RCTM_Break, time=0, value=0}
         {interpMode=RCIM_Cubic, leaveTangent=-3.8707, tangentMode=RCTM_Break, time=0.2, value=1}
         {tangentMode=RCTM_User, time=1, value=0}
        ]}}}
     Scale Curve=10
    )
    Index 0 Param 4=&Waveform(
     Global Phase=&Subtract_Float(A=$NPC.FortniteNPC.RealGameTime, B=$Particles.SpawnStartTime)
     Global Amplitude Scale=2
     "Waveform [1]"=::NewEnumerator12
     "[1] Phase"=&Subtract_Float(A=$NPC.FortniteNPC.RealGameTime, B=$Particles.SpawnStartTime)
     "[1] Frequency"=13
     "[1] Amplitude Min/Max"=<0.5 1>
    )
    Index 1 Param 1=&MakeFloatFromVector(Vector=&MakeVectorFromLinearColorRGB(Color=$Particles.Color_ParticleEnergy), Channel=::NewEnumerator0)
    Index 1 Param 2=&MakeFloatFromVector(Vector=&MakeVectorFromLinearColorRGB(Color=$Particles.Color_ParticleEnergy), Channel=::NewEnumerator1)
    Index 1 Param 3=&MakeFloatFromVector(Vector=&MakeVectorFromLinearColorRGB(Color=$Particles.Color_ParticleEnergy), Channel=::NewEnumerator2)
    Index 1 Param 4=&FloatFromCurve(FloatCurve=DataInterface() {propertyValues {Curve {keys [
         {interpMode=RCIM_Cubic, leaveTangent=0.099, tangentMode=RCTM_Break, time=0, value=0}
         {arriveTangent=1.6406, tangentMode=RCTM_User, time=1, value=1.2}
        ]}}})
    Index 2 Param 1=100
    Index 2 Param 2=550
    Index 2 Param 3=&FloatFromCurve(FloatCurve=DataInterface() {propertyValues {Curve {keys [<at 0 0.5>,<at 1 1>]}}}, CurveIndex=$Particles.TelAge, Scale Curve=300)
    Index 2 Param 4=&FloatFromCurve(FloatCurve=DataInterface() {propertyValues {Curve {keys [
         {interpMode=RCIM_Cubic, tangentMode=RCTM_Break, time=0.5, value=0}
         {arriveTangent=2.6267, interpMode=RCIM_Cubic, tangentMode=RCTM_Break, time=1, value=0.8}
        ]}}})
    Index 3 Param 1=1
    Index 3 Param 2=&MakeFloatFromVector(Vector=Linked() {linkedVariable {name=Particles.ManagePos, type=NiagaraPosition}}, Channel=::NewEnumerator0)
    Index 3 Param 3=&MakeFloatFromVector(Vector=Linked() {linkedVariable {name=Particles.ManagePos, type=NiagaraPosition}}, Channel=::NewEnumerator1)
    Index 3 Param 4=&MakeFloatFromVector(Vector=Linked() {linkedVariable {name=Particles.ManagePos, type=NiagaraPosition}}, Channel=::NewEnumerator2)
    Module.Index0Param0Write=true
    Module.Index0Param1Write=true
    Module.Index0Param2Write=true
    Module.Index0Param3Write=true
    Module.Index1Param0Write=true
    Module.Index1Param1Write=true
    Module.Index1Param2Write=true
    Module.Index1Param3Write=true
    Module.Index2Param0Write=true
    Module.Index2Param1Write=true
    Module.Index2Param2Write=true
    Module.Index2Param3Write=true
    Module.Index3Param0Write=true
    Module.Index3Param1Write=true
    Module.Index3Param2Write=true
    Module.Index3Param3Write=true
    Write Parameter Index 1=true
    Write Parameter Index 2=true
    Write Parameter Index 3=true
   )
   &ScaleColor(
    Scale Alpha=&FloatFromCurve(FloatCurve=DataInterface() {propertyValues {Curve {keys [
         {interpMode=RCIM_Cubic, leaveTangent=0.0758, tangentMode=RCTM_Break, time=0, value=0}
         {interpMode=RCIM_Cubic, tangentMode=RCTM_Break, time=0.35, value=1}
         {interpMode=RCIM_Cubic, leaveTangent=-0.0057, tangentMode=RCTM_Break, time=0.5, value=1}
         {arriveTangent=-4.4597, interpMode=RCIM_Cubic, leaveTangent=-4.4597, tangentMode=RCTM_User, time=1, value=0}
        ]}}})
    ScaleA=true
    ScaleRGB=true
   )
   &ScaleColor(
    Scale Alpha=&FloatFromCurve(
     FloatCurve=DataInterface() {propertyValues {Curve {keys [
         {interpMode=RCIM_Cubic, leaveTangent=0.0894, tangentMode=RCTM_Break, time=0, value=0}
         {arriveTangent=-0.2807, interpMode=RCIM_Cubic, tangentMode=RCTM_Break, time=0.7, value=1}
         {arriveTangent=-0.0632, interpMode=RCIM_Cubic, leaveTangent=-0.0632, tangentMode=RCTM_User, time=1, value=1}
         <at 1.1 0 RCIM_Cubic>
        ]}}}
     CurveIndex=$Particles.TelAge
    )
    ScaleA=true
    ScaleRGB=true
   ) {module=ScaleColor001}
  ]
  rendererClasses [&NiagaraSpriteRendererProperties]
  renderers [{rendererClass=&NiagaraSpriteRendererProperties}]
 }
}
summary {
 systemName=NS_FirePetal_WanderingBoss_Bombard_AOE_Idle_ManageVersion
 emitters [
  {emitterName=Decal, rendererClasses [&NiagaraDecalRendererProperties, &NiagaraMeshRendererProperties]}
  {emitterName=Smoke_Ring001, rendererClasses [&NiagaraSpriteRendererProperties]}
  {
   emitterName=GlowAndEnergy
   rendererClasses [
    &NiagaraSpriteRendererProperties
    &NiagaraLightRendererProperties
    &NiagaraMeshRendererProperties
   ]
  }
  {emitterName=FalldownGlow, rendererClasses [&NiagaraSpriteRendererProperties]}
 ]
 userVariables [
  User.bIsFrontEnd {type=NiagaraBool}
  User.bIsFrontEndPreview {type=NiagaraBool}
  User.Color Variant {type=NiagaraInt32}
  User.Duration {type=NiagaraFloat, defaultValue=3, description=""}
  User.ExplodeArray {
   type=NiagaraDataInterfaceArrayBool
   description=""
   defaultValue {
    struct=NiagaraExt_VariableValue_DataInterface
    value {
     dataInterface=&self:NiagaraDataInterfaceArrayBool_0
     dataInterfaceClass=/Script/Niagara.NiagaraDataInterfaceArrayBool
    }
   }
  }
  User.Radius {type=NiagaraFloat, defaultValue=100, description=""}
  User.SourcePlaneSize {type=Vector3f, description="", defaultValue {struct=Vector3f, value=<50 50 0>}}
  User.SpawnArray {
   type=NiagaraDataInterfaceArrayPosition
   description=""
   defaultValue {
    struct=NiagaraExt_VariableValue_DataInterface
    value {
     dataInterface=&self:NiagaraDataInterfaceArrayPosition_0
     dataInterfaceClass=&NiagaraDataInterfaceArrayPosition
    }
   }
  }
  User.SpawnCounter {
   type=NiagaraDataInterfaceSimpleCounter
   description=""
   defaultValue {
    struct=NiagaraExt_VariableValue_DataInterface
    value {
     dataInterface=&self:NiagaraDataInterfaceSimpleCounter_0
     dataInterfaceClass=/Script/Niagara.NiagaraDataInterfaceSimpleCounter
    }
   }
  }
  User.SpawnRateList {
   type=NiagaraDataInterfaceArrayPosition
   description=""
   defaultValue {
    struct=NiagaraExt_VariableValue_DataInterface
    value {
     dataInterface=&self:NiagaraDataInterfaceArrayPosition_1
     dataInterfaceClass=&NiagaraDataInterfaceArrayPosition
    }
   }
  }
  User.SpawnRateStartTime {
   type=NiagaraDataInterfaceArrayFloat
   defaultValue {
    struct=NiagaraExt_VariableValue_DataInterface
    value {
     dataInterface=&self:NiagaraDataInterfaceArrayFloat_0
     dataInterfaceClass=&NiagaraDataInterfaceArrayFloat
    }
   }
  }
 ]
}
